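# Translucent, lit character material for the Lady Atlantis female body, usable on skeletal meshes with cloth, morph targets and mesh deformers.
# The shared character PreFX function feeds the MF_Slurp weapon-skin effect, with the SkinFX_Mask texture as its alpha.
# A scalar Opacity parameter then overrides the opacity attribute before the shared PostFX function.
# Lighting uses volumetric per-vertex directional translucency, and the mask defaults to an Axl backpack FX texture.
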Material /BRCosmetics/Characters/Player/Female/Medium/Bodies/F_MED_Lady_Atlantis/Materials/F_Med_LadyAtlantis_Translucent
graph {
 properties {
  bUseMaterialAttributes=true
  bUsedWithClothing=true
  bUsedWithMeshDeformer=true
  bUsedWithMorphTargets=true
  bUsedWithSkeletalMesh=true
  blendMode=BLEND_Translucent
  pixelDepthOffsetMode=PDOM_Legacy
  translucencyLightingMode=TLM_VolumetricPerVertexDirectional
  usedShadingModels=Default Lit
 }
 nodes [
  n1 = MaterialFunctionCall() {materialFunction=&BaseFN_CharacterFunction_PreFX}
  n2 = TextureObjectParameter() {
   group=Skin FX
   parameterName=SkinFX_Mask
   samplerType=SAMPLERTYPE_Masks
   texture=/BRCosmetics/Accessories/FORT_Backpacks/Backpack_M_MED_Axl/Textures/T_M_MED_Axl_Backpack_FX
  }
  n3 = TextureSample(Tex=@n2)
  MaterialFunctionCall(
   AO=@n1.AO
   Distance Blend=@n1.AO
   Fresnel=@n1.AO
   Gmap=@n1.AO
   Pre Skinned Local Position=@n1.AO
   Material=SetMaterialAttributes!(
    MaterialAttributes=MaterialFunctionCall!(Alpha=@n3.RGB, AO=@n1.AO, BubblesMask, SlurpNormalDistortionMask, Pre Skinned Position=@n1.AO, Material=@n1.AO) {materialFunction=/Game/Weapons/WeaponSkins/MaterialFunction/MF_Slurp}.Result
    Opacity=ScalarParameter() {defaultValue=0.5, parameterName=Opacity}
   ) {attributeSetTypes [#1]}
  ) {materialFunction=&BaseFN_CharacterFunction_PostFX}
 ]
}
