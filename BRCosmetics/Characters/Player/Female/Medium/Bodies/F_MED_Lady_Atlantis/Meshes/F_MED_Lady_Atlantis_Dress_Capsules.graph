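# Physics asset for the Lady Atlantis dress on the medium female body, giving the skirt bone chains simulated motion and collision shapes.
# Capsules on pelvis, thighs and calves cover the lower body, while small driver bodies anchor the skirt chains to the pelvis with free rotation.
# Five-segment side and two back chains per side, plus short hip chains, use tight angular limits.
# Cross-constraints tie neighboring side and back chains into a connected sheet.
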PhysicsAsset /BRCosmetics/Characters/Player/Female/Medium/Bodies/F_MED_Lady_Atlantis/Meshes/F_MED_Lady_Atlantis_Dress_Capsules
bodies {
 calf_l [{shapeType=Capsule, length=43.9788, radius=7.3601, center=<-26.3577 -4.7593 -0.4554>, rotation=<rot 89.2595 -179.9999 -179.9999>}]
 calf_r [{shapeType=Capsule, length=43.2826, radius=6.6179, center=<25.474 4.5933 -0.0558>, rotation=<rot 88.474 175.7292 -179.9999>}]
 dyn_driver_skirt_bckA_l [{shapeType=Capsule, length=0.505, radius=0.505, rotation=<rot 0 -90 0>}]
 dyn_driver_skirt_bckA_r [{shapeType=Capsule, length=0.505, radius=0.505, rotation=<rot 0 -90 0>}]
 dyn_driver_skirt_bck_01_l [{shapeType=Box, extentX=1.01, extentY=1.01, extentZ=1.01, rotation=<rot 0 -90 0>}]
 dyn_driver_skirt_bck_01_r [{shapeType=Box, extentX=1.01, extentY=1.01, extentZ=1.01, rotation=<rot 0 -90 0>}]
 dyn_driver_skirt_hip_01_l [{shapeType=Capsule, length=0.505, radius=0.505, rotation=<rot 0 -90 0>}]
 dyn_driver_skirt_hip_01_r [{shapeType=Capsule, length=0.505, radius=0.505, rotation=<rot 0 -90 0>}]
 dyn_driver_skirt_side_01_l [{shapeType=Box, extentX=1.01, extentY=1.01, extentZ=1.01, rotation=<rot 0 -90 0>}]
 dyn_driver_skirt_side_01_r [{shapeType=Box, extentX=1.01, extentY=1.01, extentZ=1.01, rotation=<rot 0 -90 0>}]
 dyn_skirt_bckA_01_l [{shapeType=Capsule, length=15, radius=10, center=<6 0 9.5>, rotation=<rot 90 0 0>}]
 dyn_skirt_bckA_01_r [{shapeType=Capsule, length=15, radius=10, center=<6 0 9.5>, rotation=<rot 90 0 0>}]
 dyn_skirt_bckA_02_l [{shapeType=Capsule, length=15, radius=10, center=<6 0 9.5>, rotation=<rot 90 0 0>}]
 dyn_skirt_bckA_02_r [{shapeType=Capsule, length=15, radius=10, center=<6 0 9.5>, rotation=<rot 90 0 0>}]
 dyn_skirt_bckA_03_l [{shapeType=Capsule, length=15, radius=10, center=<6 0 9.5>, rotation=<rot 90 0 0>}]
 dyn_skirt_bckA_03_r [{shapeType=Capsule, length=15, radius=10, center=<6 0 9.5>, rotation=<rot 90 0 0>}]
 dyn_skirt_bckA_04_l [{shapeType=Capsule, length=15, radius=10, center=<6 0 9.5>, rotation=<rot 90 0 0>}]
 dyn_skirt_bckA_04_r [{shapeType=Capsule, length=15, radius=10, center=<6 0 9.5>, rotation=<rot 90 0 0>}]
 dyn_skirt_bckA_05_l [{shapeType=Capsule, length=10, radius=8, center=<5 0 7>, rotation=<rot 90 0 0>}]
 dyn_skirt_bckA_05_r [{shapeType=Capsule, length=10, radius=8, center=<5 0 7>, rotation=<rot 90 0 0>}]
 dyn_skirt_bck_01_l [{shapeType=Capsule, length=15, radius=10, center=<7 0 9.5>, rotation=<rot 90 0 0>}]
 dyn_skirt_bck_01_r [{shapeType=Capsule, length=15, radius=10, center=<7 0 9.5>, rotation=<rot 90 0 0>}]
 dyn_skirt_bck_02_l [{shapeType=Capsule, length=15, radius=10, center=<6 0 9.5>, rotation=<rot 90 0 0>}]
 dyn_skirt_bck_02_r [{shapeType=Capsule, length=15, radius=10, center=<6 0 9.5>, rotation=<rot 90 0 0>}]
 dyn_skirt_bck_03_l [{shapeType=Capsule, length=15, radius=10, center=<6 0 9.5>, rotation=<rot 90 0 0>}]
 dyn_skirt_bck_03_r [{shapeType=Capsule, length=15, radius=7, center=<6 0 9.5>, rotation=<rot 90 0 0>}]
 dyn_skirt_bck_04_l [{shapeType=Capsule, length=15, radius=10, center=<6 0 9.5>, rotation=<rot 90 0 0>}]
 dyn_skirt_bck_04_r [{shapeType=Capsule, length=15, radius=10, center=<6 0 9.5>, rotation=<rot 90 0 0>}]
 dyn_skirt_bck_05_l [{shapeType=Capsule, length=10, radius=8, center=<5 0 7>, rotation=<rot 90 0 0>}]
 dyn_skirt_bck_05_r [{shapeType=Capsule, length=10, radius=8, center=<5 0 7>, rotation=<rot 90 0 0>}]
 dyn_skirt_hip_01_l [{shapeType=Capsule, length=5, radius=1.9, center=<0.25 0 2>, rotation=<rot 90 0 0>}]
 dyn_skirt_hip_01_r [{shapeType=Capsule, length=5, radius=1.9, center=<-0.25 0 -2>, rotation=<rot 90 0 0>}]
 dyn_skirt_hip_02_l [{shapeType=Capsule, length=5, radius=2.55, center=<0.7 0 1.5>, rotation=<rot 90 0 0>}]
 dyn_skirt_hip_02_r [{shapeType=Capsule, length=5, radius=2.55, center=<-0.75 0 -1.5>, rotation=<rot 90 0 0>}]
 dyn_skirt_hip_03_l [{shapeType=Capsule, length=4, radius=5.05, center=<2.25 -4.5 -1.9>, rotation=<rot 0 0 15.8352>}]
 dyn_skirt_hip_03_r [{shapeType=Capsule, length=4, radius=5.05, center=<-2.25 4.5 1.9>, rotation=<rot 0 0 18.5122>}]
 dyn_skirt_side_01_l [{shapeType=Capsule, length=15, radius=10, center=<7 0 9.5>, rotation=<rot 90 0 0>}]
 dyn_skirt_side_01_r [{shapeType=Capsule, length=15, radius=10, center=<7 0 9.5>, rotation=<rot 90 0 0>}]
 dyn_skirt_side_02_l [{shapeType=Capsule, length=15, radius=10, center=<6 0 9.5>, rotation=<rot 90 0 0>}]
 dyn_skirt_side_02_r [{shapeType=Capsule, length=15, radius=10, center=<6 0 9.5>, rotation=<rot 90 0 0>}]
 dyn_skirt_side_03_l [{shapeType=Capsule, length=15, radius=10, center=<6 0 9.5>, rotation=<rot 90 0 0>}]
 dyn_skirt_side_03_r [{shapeType=Capsule, length=15, radius=10, center=<6 0 9.5>, rotation=<rot 90 0 0>}]
 dyn_skirt_side_04_l [{shapeType=Capsule, length=15, radius=10, center=<6 0 9.5>, rotation=<rot 90 0 0>}]
 dyn_skirt_side_04_r [{shapeType=Capsule, length=15, radius=9, center=<6 0 9.5>, rotation=<rot 90 0 0>}]
 dyn_skirt_side_05_l [{shapeType=Capsule, length=10, radius=8, center=<3 0 7>, rotation=<rot 90 0 0>}]
 dyn_skirt_side_05_r [{shapeType=Capsule, length=10, radius=8, center=<3 0 7>, rotation=<rot 90 0 0>}]
 pelvis [{shapeType=Capsule, length=9.1919, radius=9.9429, center=<-4.2064 -2.0929 -0.0536>, rotation=<rot -3.0486 -90 -0.0239>}]
 thigh_l [
  {shapeType=Capsule, length=36.0217, radius=7.3659, center=<-22.5185 -0.3293 0.5757>, rotation=<rot 88.2403 -179.9998 -179.9998>}
  {shapeType=Capsule, length=11.0217, radius=7.3659, center=<-12.3309 0.378 0.8761>, rotation=<rot 87.4956 -179.9999 -177.6958>}
 ]
 thigh_r [{shapeType=Capsule, length=34.8049, radius=10, center=<24.8277 -0.0864 -0.0662>, rotation=<rot 88.6552 -180 -180>}]
}
constraints [
 {bone1Name=thigh_l, bone2Name=pelvis, swing1LimitDegrees=45, swing2LimitDegrees=45, twistLimitDegrees=45}
 {bone1Name=calf_l, bone2Name=thigh_l, swing1LimitDegrees=45, swing2LimitDegrees=45, twistLimitDegrees=45}
 {bone1Name=thigh_r, bone2Name=pelvis, swing1LimitDegrees=45, swing2LimitDegrees=45, twistLimitDegrees=45}
 {bone1Name=calf_r, bone2Name=thigh_r, swing1LimitDegrees=45, swing2LimitDegrees=45, twistLimitDegrees=45}
 {
  bone1Name=dyn_driver_skirt_side_01_l
  bone2Name=pelvis
  swing1LimitDegrees=45
  swing1Motion=Free
  swing2LimitDegrees=45
  swing2Motion=Free
  twistLimitDegrees=45
  twistMotion=Free
 }
 {
  bone1Name=dyn_driver_skirt_side_01_r
  bone2Name=pelvis
  swing1LimitDegrees=45
  swing1Motion=Free
  swing2LimitDegrees=45
  swing2Motion=Free
  twistLimitDegrees=45
  twistMotion=Free
 }
 {
  bone1Name=dyn_driver_skirt_bck_01_l
  bone2Name=pelvis
  swing1LimitDegrees=45
  swing1Motion=Free
  swing2LimitDegrees=45
  swing2Motion=Free
  twistLimitDegrees=45
  twistMotion=Free
 }
 {
  bone1Name=dyn_driver_skirt_bck_01_r
  bone2Name=pelvis
  swing1LimitDegrees=45
  swing1Motion=Free
  swing2LimitDegrees=45
  swing2Motion=Free
  twistLimitDegrees=45
  twistMotion=Free
 }
 {bone1Name=dyn_skirt_side_01_l, bone2Name=dyn_driver_skirt_side_01_l, swing1LimitDegrees=20, swing2LimitDegrees=20, twistLimitDegrees=5}
 {bone1Name=dyn_skirt_side_02_l, bone2Name=dyn_skirt_side_01_l, swing1LimitDegrees=20, swing2LimitDegrees=20, twistLimitDegrees=5}
 {bone1Name=dyn_skirt_side_03_l, bone2Name=dyn_skirt_side_02_l, swing1LimitDegrees=30, swing2LimitDegrees=30, twistLimitDegrees=3}
 {bone1Name=dyn_skirt_side_04_l, bone2Name=dyn_skirt_side_03_l, swing1LimitDegrees=30, swing2LimitDegrees=30, twistLimitDegrees=3}
 {bone1Name=dyn_skirt_side_05_l, bone2Name=dyn_skirt_side_04_l, swing1LimitDegrees=30, swing2LimitDegrees=30, twistLimitDegrees=3}
 {bone1Name=dyn_skirt_side_01_r, bone2Name=dyn_driver_skirt_side_01_r, swing1LimitDegrees=20, swing2LimitDegrees=20, twistLimitDegrees=5}
 {bone1Name=dyn_skirt_side_02_r, bone2Name=dyn_skirt_side_01_r, swing1LimitDegrees=35, swing2LimitDegrees=45, twistLimitDegrees=5}
 {bone1Name=dyn_skirt_side_03_r, bone2Name=dyn_skirt_side_02_r, swing1LimitDegrees=35, swing2LimitDegrees=35, twistLimitDegrees=3}
 {bone1Name=dyn_skirt_side_04_r, bone2Name=dyn_skirt_side_03_r, swing1LimitDegrees=35, swing2LimitDegrees=35, twistLimitDegrees=3}
 {bone1Name=dyn_skirt_side_05_r, bone2Name=dyn_skirt_side_04_r, swing1LimitDegrees=35, swing2LimitDegrees=35, twistLimitDegrees=3}
 {bone1Name=dyn_skirt_bck_01_l, bone2Name=dyn_driver_skirt_bck_01_l, swing1LimitDegrees=20, swing2LimitDegrees=20, twistLimitDegrees=5}
 {bone1Name=dyn_skirt_bck_02_l, bone2Name=dyn_skirt_bck_01_l, swing1LimitDegrees=20, swing2LimitDegrees=20, twistLimitDegrees=5}
 {bone1Name=dyn_skirt_bck_03_l, bone2Name=dyn_skirt_bck_02_l, swing1LimitDegrees=30, swing2LimitDegrees=30, twistLimitDegrees=3}
 {bone1Name=dyn_skirt_bck_04_l, bone2Name=dyn_skirt_bck_03_l, swing1LimitDegrees=30, swing2LimitDegrees=30, twistLimitDegrees=3}
 {bone1Name=dyn_skirt_bck_05_l, bone2Name=dyn_skirt_bck_04_l, swing1LimitDegrees=30, swing2LimitDegrees=30, twistLimitDegrees=3}
 {bone1Name=dyn_skirt_bck_01_r, bone2Name=dyn_driver_skirt_bck_01_r, swing1LimitDegrees=20, swing2LimitDegrees=20, twistLimitDegrees=5}
 {bone1Name=dyn_skirt_bck_02_r, bone2Name=dyn_skirt_bck_01_r, swing1LimitDegrees=20, swing2LimitDegrees=20, twistLimitDegrees=5}
 {bone1Name=dyn_skirt_bck_03_r, bone2Name=dyn_skirt_bck_02_r, swing1LimitDegrees=35, swing2LimitDegrees=35, twistLimitDegrees=3}
 {bone1Name=dyn_skirt_bck_04_r, bone2Name=dyn_skirt_bck_03_r, swing1LimitDegrees=35, swing2LimitDegrees=35, twistLimitDegrees=3}
 {bone1Name=dyn_skirt_bck_05_r, bone2Name=dyn_skirt_bck_04_r, swing1LimitDegrees=30, swing2LimitDegrees=30, twistLimitDegrees=3}
 {bone1Name=dyn_skirt_bck_02_r, bone2Name=dyn_skirt_side_02_r, swing1LimitDegrees=45, swing2LimitDegrees=45, twistLimitDegrees=20}
 {bone1Name=dyn_skirt_bck_05_l, bone2Name=dyn_skirt_side_05_l, swing1LimitDegrees=40, swing2LimitDegrees=40, twistLimitDegrees=10}
 {
  bone1Name=dyn_driver_skirt_bckA_l
  bone2Name=pelvis
  swing1LimitDegrees=45
  swing1Motion=Free
  swing2LimitDegrees=45
  swing2Motion=Free
  twistLimitDegrees=45
  twistMotion=Free
 }
 {
  bone1Name=dyn_driver_skirt_bckA_r
  bone2Name=pelvis
  swing1LimitDegrees=45
  swing1Motion=Free
  swing2LimitDegrees=45
  swing2Motion=Free
  twistLimitDegrees=45
  twistMotion=Free
 }
 {bone1Name=dyn_skirt_bckA_01_l, bone2Name=dyn_driver_skirt_bckA_l, swing1LimitDegrees=25, swing2LimitDegrees=25, twistLimitDegrees=5}
 {bone1Name=dyn_skirt_bckA_02_l, bone2Name=dyn_skirt_bckA_01_l, swing1LimitDegrees=25, swing2LimitDegrees=25, twistLimitDegrees=5}
 {bone1Name=dyn_skirt_bckA_03_l, bone2Name=dyn_skirt_bckA_02_l, swing1LimitDegrees=25, swing2LimitDegrees=25, twistLimitDegrees=5}
 {bone1Name=dyn_skirt_bckA_04_l, bone2Name=dyn_skirt_bckA_03_l, swing1LimitDegrees=25, swing2LimitDegrees=25, twistLimitDegrees=5}
 {bone1Name=dyn_skirt_bckA_05_l, bone2Name=dyn_skirt_bckA_04_l, swing1LimitDegrees=25, swing2LimitDegrees=25, twistLimitDegrees=5}
 {bone1Name=dyn_skirt_bckA_01_r, bone2Name=dyn_driver_skirt_bckA_r, swing1LimitDegrees=25, swing2LimitDegrees=25, twistLimitDegrees=5}
 {bone1Name=dyn_skirt_bckA_02_r, bone2Name=dyn_skirt_bckA_01_r, swing1LimitDegrees=25, swing2LimitDegrees=25, twistLimitDegrees=5}
 {bone1Name=dyn_skirt_bckA_03_r, bone2Name=dyn_skirt_bckA_02_r, swing1LimitDegrees=25, swing2LimitDegrees=25, twistLimitDegrees=5}
 {bone1Name=dyn_skirt_bckA_04_r, bone2Name=dyn_skirt_bckA_03_r, swing1LimitDegrees=25, swing2LimitDegrees=25, twistLimitDegrees=5}
 {bone1Name=dyn_skirt_bckA_05_r, bone2Name=dyn_skirt_bckA_04_r, swing1LimitDegrees=25, swing2LimitDegrees=25, twistLimitDegrees=5}
 {bone1Name=dyn_skirt_bckA_02_r, bone2Name=dyn_skirt_bckA_02_l, swing1LimitDegrees=25, swing2LimitDegrees=25, twistLimitDegrees=10}
 {bone1Name=dyn_skirt_bckA_03_r, bone2Name=dyn_skirt_bckA_03_l, swing1LimitDegrees=25, swing2LimitDegrees=25, twistLimitDegrees=10}
 {bone1Name=dyn_skirt_bckA_04_r, bone2Name=dyn_skirt_bckA_04_l, swing1LimitDegrees=25, swing2LimitDegrees=25, twistLimitDegrees=10}
 {bone1Name=dyn_skirt_bckA_02_l, bone2Name=dyn_skirt_bck_02_l, swing1LimitDegrees=25, swing2LimitDegrees=25, twistLimitDegrees=10}
 {bone1Name=dyn_skirt_bckA_03_l, bone2Name=dyn_skirt_bck_03_l, swing1LimitDegrees=25, swing2LimitDegrees=25, twistLimitDegrees=10}
 {bone1Name=dyn_skirt_bckA_04_l, bone2Name=dyn_skirt_bck_04_l, swing1LimitDegrees=25, swing2LimitDegrees=25, twistLimitDegrees=10}
 {bone1Name=dyn_skirt_bckA_02_r, bone2Name=dyn_skirt_bck_02_r, swing1LimitDegrees=25, swing2LimitDegrees=25, twistLimitDegrees=10}
 {bone1Name=dyn_skirt_bckA_03_r, bone2Name=dyn_skirt_bck_03_r, swing1LimitDegrees=25, swing2LimitDegrees=25, twistLimitDegrees=10}
 {bone1Name=dyn_skirt_bckA_04_r, bone2Name=dyn_skirt_bck_04_r, swing1LimitDegrees=25, swing2LimitDegrees=25, twistLimitDegrees=10}
 {bone1Name=dyn_skirt_bck_03_r, bone2Name=dyn_skirt_side_03_r, swing1LimitDegrees=45, swing2LimitDegrees=45, twistLimitDegrees=20}
 {bone1Name=dyn_skirt_bck_04_r, bone2Name=dyn_skirt_side_04_r, swing1LimitDegrees=45, swing2LimitDegrees=45, twistLimitDegrees=20}
 {bone1Name=dyn_skirt_bck_02_l, bone2Name=dyn_skirt_side_02_l, swing1LimitDegrees=45, swing2LimitDegrees=45, twistLimitDegrees=20}
 {bone1Name=dyn_skirt_bck_03_l, bone2Name=dyn_skirt_side_03_l, swing1LimitDegrees=45, swing2LimitDegrees=45, twistLimitDegrees=20}
 {bone1Name=dyn_skirt_bck_04_l, bone2Name=dyn_skirt_side_04_l, swing1LimitDegrees=45, swing2LimitDegrees=45, twistLimitDegrees=20}
 {
  bone1Name=dyn_driver_skirt_hip_01_l
  bone2Name=pelvis
  swing1LimitDegrees=45
  swing1Motion=Free
  swing2LimitDegrees=45
  swing2Motion=Free
  twistLimitDegrees=45
  twistMotion=Free
 }
 {
  bone1Name=dyn_driver_skirt_hip_01_r
  bone2Name=pelvis
  swing1LimitDegrees=45
  swing1Motion=Free
  swing2LimitDegrees=45
  swing2Motion=Free
  twistLimitDegrees=45
  twistMotion=Free
 }
 {bone1Name=dyn_skirt_hip_01_l, bone2Name=dyn_driver_skirt_hip_01_l, swing1LimitDegrees=5, swing2LimitDegrees=5, twistLimitDegrees=5}
 {bone1Name=dyn_skirt_hip_02_l, bone2Name=dyn_skirt_hip_01_l, swing1LimitDegrees=20, swing2LimitDegrees=20, twistLimitDegrees=5}
 {bone1Name=dyn_skirt_hip_03_l, bone2Name=dyn_skirt_hip_02_l, swing1LimitDegrees=15, swing2LimitDegrees=15, twistLimitDegrees=5}
 {bone1Name=dyn_skirt_hip_01_r, bone2Name=dyn_driver_skirt_hip_01_r, swing1LimitDegrees=5, swing2LimitDegrees=5, twistLimitDegrees=5}
 {bone1Name=dyn_skirt_hip_03_r, bone2Name=dyn_skirt_hip_02_r, swing1LimitDegrees=15, swing2LimitDegrees=15, twistLimitDegrees=5}
 {bone1Name=dyn_skirt_hip_02_r, bone2Name=dyn_skirt_hip_01_r, swing1LimitDegrees=20, swing2LimitDegrees=20, twistLimitDegrees=5}
]
