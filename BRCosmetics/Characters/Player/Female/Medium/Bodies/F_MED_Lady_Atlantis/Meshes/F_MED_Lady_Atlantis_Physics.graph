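# Physics asset for the Lady Atlantis medium female player body cosmetic, covering secondary motion of the fins on both forearms or arms and calves.
# Each limb has three fins (A, B, C), each a single capsule body constrained to its own small capsule on a matching driver bone.
# Swing and twist limits keep the fins close to their drivers, with calf fins allowed more movement than arm fins.
# No other body parts have collision shapes here.
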PhysicsAsset /BRCosmetics/Characters/Player/Female/Medium/Bodies/F_MED_Lady_Atlantis/Meshes/F_MED_Lady_Atlantis_Physics
bodies {
 dyn_driver_fin_armA_l [{shapeType=Capsule, length=1, radius=1}]
 dyn_driver_fin_armA_r [{shapeType=Capsule, length=1, radius=1}]
 dyn_driver_fin_armB_l [{shapeType=Capsule, length=1, radius=1}]
 dyn_driver_fin_armB_r [{shapeType=Capsule, length=1, radius=1}]
 dyn_driver_fin_armC_l [{shapeType=Capsule, length=1, radius=1}]
 dyn_driver_fin_armC_r [{shapeType=Capsule, length=1, radius=1}]
 dyn_driver_fin_calfA_l [{shapeType=Capsule, length=0.505, radius=0.505, rotation=<rot 0 -90 0>}]
 dyn_driver_fin_calfA_r [{shapeType=Capsule, length=0.505, radius=0.505, rotation=<rot 0 -90 0>}]
 dyn_driver_fin_calfB_l [{shapeType=Capsule, length=0.505, radius=0.505, rotation=<rot 0 -90 0>}]
 dyn_driver_fin_calfB_r [{shapeType=Capsule, length=0.505, radius=0.505, rotation=<rot 0 -90 0>}]
 dyn_driver_fin_calfC_l [{shapeType=Capsule, length=0.505, radius=0.505, rotation=<rot 0 -90 0>}]
 dyn_driver_fin_calfC_r [{shapeType=Capsule, length=0.505, radius=0.505, rotation=<rot 0 -90 0>}]
 dyn_fin_armA_01_l [{shapeType=Capsule, length=4.5, radius=1.75, center=<3 0 0>, rotation=<rot 90 0 0>}]
 dyn_fin_armA_01_r [{shapeType=Capsule, length=3.75, radius=2.5, center=<3 0 0>, rotation=<rot 90 0 0>}]
 dyn_fin_armB_01_l [{shapeType=Capsule, length=4, radius=3, center=<3 0 0>, rotation=<rot 90 0 0>}]
 dyn_fin_armB_01_r [{shapeType=Capsule, length=4, radius=2.5, center=<3 0 0>, rotation=<rot 90 0 0>}]
 dyn_fin_armC_01_l [{shapeType=Capsule, length=3.75, radius=1.75, center=<3 0 0>, rotation=<rot 90 0 0>}]
 dyn_fin_armC_01_r [{shapeType=Capsule, length=4.25, radius=2, center=<3 0 0>, rotation=<rot 90 0 0>}]
 dyn_fin_calfA_01_l [{shapeType=Capsule, length=5.25, radius=2.5, center=<3 0 0>, rotation=<rot 90 0 0>}]
 dyn_fin_calfA_01_r [{shapeType=Capsule, length=4.75, radius=2.75, center=<5 0 0>, rotation=<rot 90 0 0>}]
 dyn_fin_calfB_01_l [{shapeType=Capsule, length=6.25, radius=3.25, center=<4 0 0>, rotation=<rot 90 0 0>}]
 dyn_fin_calfB_01_r [{shapeType=Capsule, length=6.25, radius=3.25, center=<4 0 0>, rotation=<rot 90 0 0>}]
 dyn_fin_calfC_01_l [{shapeType=Capsule, length=6.25, radius=3.25, center=<4 0 0>, rotation=<rot 90 0 0>}]
 dyn_fin_calfC_01_r [{shapeType=Capsule, length=6.25, radius=3.25, center=<4 0 0>, rotation=<rot 90 0 0>}]
}
constraints [
 {bone1Name=dyn_fin_armA_01_l, bone2Name=dyn_driver_fin_armA_l, swing1LimitDegrees=22, swing2LimitDegrees=22, twistLimitDegrees=2}
 {bone1Name=dyn_fin_armB_01_l, bone2Name=dyn_driver_fin_armB_l, swing1LimitDegrees=15, swing2LimitDegrees=15, twistLimitDegrees=2}
 {bone1Name=dyn_fin_armC_01_l, bone2Name=dyn_driver_fin_armC_l, swing1LimitDegrees=18, swing2LimitDegrees=18, twistLimitDegrees=2}
 {bone1Name=dyn_fin_armA_01_r, bone2Name=dyn_driver_fin_armA_r, swing1LimitDegrees=20, swing2LimitDegrees=20, twistLimitDegrees=2}
 {bone1Name=dyn_fin_armB_01_r, bone2Name=dyn_driver_fin_armB_r, swing1LimitDegrees=15, swing2LimitDegrees=15, twistLimitDegrees=2}
 {bone1Name=dyn_fin_armC_01_r, bone2Name=dyn_driver_fin_armC_r, swing1LimitDegrees=18, swing2LimitDegrees=18, twistLimitDegrees=2}
 {bone1Name=dyn_fin_calfA_01_l, bone2Name=dyn_driver_fin_calfA_l, swing1LimitDegrees=30, swing2LimitDegrees=30, twistLimitDegrees=5}
 {bone1Name=dyn_fin_calfB_01_l, bone2Name=dyn_driver_fin_calfB_l, swing1LimitDegrees=25, swing2LimitDegrees=25, twistLimitDegrees=5}
 {bone1Name=dyn_fin_calfC_01_l, bone2Name=dyn_driver_fin_calfC_l, swing1LimitDegrees=25, swing2LimitDegrees=25, twistLimitDegrees=5}
 {bone1Name=dyn_fin_calfA_01_r, bone2Name=dyn_driver_fin_calfA_r, swing1LimitDegrees=25, swing2LimitDegrees=25, twistLimitDegrees=5}
 {bone1Name=dyn_fin_calfB_01_r, bone2Name=dyn_driver_fin_calfB_r, swing1LimitDegrees=25, swing2LimitDegrees=25, twistLimitDegrees=5}
 {bone1Name=dyn_fin_calfC_01_r, bone2Name=dyn_driver_fin_calfC_r, swing1LimitDegrees=25, swing2LimitDegrees=25, twistLimitDegrees=5}
]
