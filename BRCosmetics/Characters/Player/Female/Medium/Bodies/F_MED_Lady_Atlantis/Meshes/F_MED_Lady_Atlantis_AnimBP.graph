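# Secondary-motion AnimBlueprint for the Lady Atlantis cosmetic body, animating skirt and arm/calf fins over a pose copied from the parent mesh.
# Thigh rotation alphas drive skirt corrective bones before and after dress RigidBody simulation.
# Skydiving switches to a pelvis-space simulation with leg constraints and offsets; AnimDynamics then moves the fins.
# Event graph also sets skydiving gravity and force variables, which no node reads; several AnimGraph nodes are disconnected.
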AnimBlueprint /BRCosmetics/Characters/Player/Female/Medium/Bodies/F_MED_Lady_Atlantis/Meshes/F_MED_Lady_Atlantis_AnimBP
target_skeleton=&/F_MED_Lady_Atlantis_Skeleton
graphs [
 AnimGraph {nodes [
   n1 = Anim Dynamics - Bone: dyn_fin_armA_01_l(Alpha=0.3, ExternalForce="0,0,0") {
    bGravityOverrideInSimSpace=true
    bUseGravityOverride=true
    gravityScale=0.3
    lODThreshold=1
    simulationSpace=BoneRelative
    alphaBoolBlend {blendOption=Linear}
    alphaScaleBiasClamp {clampMax=0.09, clampMin=0.05}
    boundBone {boneName=dyn_fin_armA_01_l}
    gravityOverride {z=-980}
    physicsBodyDefinitions [{
      sphereCollisionRadius=0
      boundBone {boneName=dyn_fin_armA_01_l}
      boxExtents=<2 5 5>
      constraintSetup {angularLimitsMax=<0 0 0>, angularLimitsMin=<0 0 0>, angularTarget=<0 0 0>, linearAxesMax=<0.5 0.5 0.5>, linearAxesMin=<-1.5 -1.5 -1.5>}
      localJointOffset=<5 0 0>
     }]
    relativeSpaceBone {boneName=pelvis}
   }
   n2 = Anim Dynamics - Bone: dyn_fin_armB_01_l(ComponentPose=@n1, Alpha=0.3, ExternalForce="0,0,0") {
    bGravityOverrideInSimSpace=true
    bUseGravityOverride=true
    gravityScale=0.3
    lODThreshold=1
    simulationSpace=BoneRelative
    alphaBoolBlend {blendOption=Linear}
    alphaScaleBiasClamp {clampMax=0.09, clampMin=0.05}
    boundBone {boneName=dyn_fin_armB_01_l}
    gravityOverride {z=-980}
    physicsBodyDefinitions [{
      sphereCollisionRadius=0
      boundBone {boneName=dyn_fin_armB_01_l}
      boxExtents=<2 5 5>
      constraintSetup {angularLimitsMax=<0 0 0>, angularLimitsMin=<0 0 0>, angularTarget=<0 0 0>, linearAxesMax=<0.5 0.5 0.5>, linearAxesMin=<-1.5 -1.5 -1.5>}
      localJointOffset=<5 0 0>
     }]
    relativeSpaceBone {boneName=pelvis}
   }
   n3 = Anim Dynamics - Bone: dyn_fin_armC_01_l(ComponentPose=@n2, Alpha=0.3, ExternalForce="0,0,0") {
    bGravityOverrideInSimSpace=true
    bUseGravityOverride=true
    gravityScale=0.3
    lODThreshold=1
    simulationSpace=BoneRelative
    alphaBoolBlend {blendOption=Linear}
    alphaScaleBiasClamp {clampMax=0.09, clampMin=0.05}
    boundBone {boneName=dyn_fin_armC_01_l}
    gravityOverride {z=-980}
    physicsBodyDefinitions [{
      sphereCollisionRadius=0
      boundBone {boneName=dyn_fin_armC_01_l}
      boxExtents=<2 5 5>
      constraintSetup {angularLimitsMax=<0 0 0>, angularLimitsMin=<0 0 0>, angularTarget=<0 0 0>, linearAxesMax=<0.5 0.5 0.5>, linearAxesMin=<-1.5 -1.5 -1.5>}
      localJointOffset=<5 0 0>
     }]
    relativeSpaceBone {boneName=pelvis}
   }
   n4 = Anim Dynamics - Bone: dyn_fin_armA_01_r(ComponentPose=@n3, Alpha=0.3, ExternalForce="0,0,0") {
    bGravityOverrideInSimSpace=true
    bUseGravityOverride=true
    gravityScale=0.3
    lODThreshold=1
    simulationSpace=BoneRelative
    alphaBoolBlend {blendOption=Linear}
    alphaScaleBiasClamp {clampMax=0.09, clampMin=0.05}
    boundBone {boneName=dyn_fin_armA_01_r}
    gravityOverride {z=-980}
    physicsBodyDefinitions [{
      sphereCollisionRadius=0
      boundBone {boneName=dyn_fin_armA_01_r}
      boxExtents=<2 5 5>
      constraintSetup {angularLimitsMax=<0 0 0>, angularLimitsMin=<0 0 0>, angularTarget=<0 0 0>, linearAxesMax=<0.5 0.5 0.5>, linearAxesMin=<-1.5 -1.5 -1.5>}
      localJointOffset=<5 0 0>
     }]
    relativeSpaceBone {boneName=pelvis}
   }
   n5 = Anim Dynamics - Bone: dyn_fin_armB_01_r(ComponentPose=@n4, Alpha=0.3, ExternalForce="0,0,0") {
    bGravityOverrideInSimSpace=true
    bUseGravityOverride=true
    gravityScale=0.3
    lODThreshold=1
    simulationSpace=BoneRelative
    alphaBoolBlend {blendOption=Linear}
    alphaScaleBiasClamp {clampMax=0.09, clampMin=0.05}
    boundBone {boneName=dyn_fin_armB_01_r}
    gravityOverride {z=-980}
    physicsBodyDefinitions [{
      sphereCollisionRadius=0
      boundBone {boneName=dyn_fin_armB_01_r}
      boxExtents=<2 5 5>
      constraintSetup {angularLimitsMax=<0 0 0>, angularLimitsMin=<0 0 0>, angularTarget=<0 0 0>, linearAxesMax=<0.5 0.5 0.5>, linearAxesMin=<-1.5 -1.5 -1.5>}
      localJointOffset=<5 0 0>
     }]
    relativeSpaceBone {boneName=pelvis}
   }
   Anim Dynamics - Bone: dyn_fin_armC_01_r(ComponentPose=@n5, Alpha=0.3, ExternalForce="0,0,0") {
    bGravityOverrideInSimSpace=true
    bUseGravityOverride=true
    gravityScale=0.3
    lODThreshold=1
    simulationSpace=BoneRelative
    alphaBoolBlend {blendOption=Linear}
    alphaScaleBiasClamp {clampMax=0.09, clampMin=0.05}
    boundBone {boneName=dyn_fin_armC_01_r}
    gravityOverride {z=-980}
    physicsBodyDefinitions [{
      sphereCollisionRadius=0
      boundBone {boneName=dyn_fin_armC_01_r}
      boxExtents=<2 5 5>
      constraintSetup {angularLimitsMax=<0 0 0>, angularLimitsMin=<0 0 0>, angularTarget=<0 0 0>, linearAxesMax=<0.5 0.5 0.5>, linearAxesMin=<-1.5 -1.5 -1.5>}
      localJointOffset=<5 0 0>
     }]
    relativeSpaceBone {boneName=pelvis}
   }
   n6 = Anim Dynamics - Bone: dyn_hair_extra_g_02(Alpha=1, ExternalForce="0,0,0") {
    angularDampingOverride=0.9999
    angularSpringConstant=25
    bAngularSpring=true
    bGravityOverrideInSimSpace=true
    bOverrideAngularDamping=true
    simulationSpace=RootRelative
    alphaBoolBlend {blendOption=Linear}
    boundBone {boneName=dyn_hair_extra_g_02}
    componentAppliedLinearAccClamp=<3500 3500 3500>
    componentLinearAccScale=<0.6 0.6 0.8>
    componentLinearVelScale=<0.6 0.6 0.8>
    physicsBodyDefinitions [{
      sphereCollisionRadius=0
      boundBone {boneName=dyn_hair_extra_g_02}
      boxExtents=<20 15 15>
      constraintSetup {angularLimitsMax=<1 -5 6>, angularLimitsMin=<-1 -45 1>, angularTarget=<0 0 0>}
      localJointOffset=<1 0 0>
     }]
    relativeSpaceBone {boneName=dyn_root_simspace}
   }
   Anim Dynamics - Bone: dyn_hair_extra_g_03(ComponentPose=@n6, Alpha=1, ExternalForce="0,0,0") {
    angularDampingOverride=0.9999
    angularSpringConstant=25
    bAngularSpring=true
    bGravityOverrideInSimSpace=true
    bOverrideAngularDamping=true
    simulationSpace=RootRelative
    alphaBoolBlend {blendOption=Linear}
    boundBone {boneName=dyn_hair_extra_g_03}
    componentAppliedLinearAccClamp=<3500 3500 3500>
    componentLinearAccScale=<0.6 0.6 0.8>
    componentLinearVelScale=<0.6 0.6 0.8>
    physicsBodyDefinitions [{
      sphereCollisionRadius=0
      boundBone {boneName=dyn_hair_extra_g_03}
      boxExtents=<20 15 15>
      constraintSetup {angularLimitsMax=<1 -5 8>, angularLimitsMin=<-1 -45 1>, angularTarget=<0 0 0>}
      localJointOffset=<1 0 0>
     }]
    relativeSpaceBone {boneName=dyn_root_simspace}
   }
   Anim Dynamics - Bone: dyn_cube_r_in_02(Alpha=0.4, ExternalForce="0,0,0") {
    angularDampingOverride=0.9
    angularSpringConstant=200000
    bAngularSpring=true
    bLinearSpring=true
    bOverrideAngularDamping=true
    bOverrideLinearDamping=true
    gravityScale=0
    linearDampingOverride=0.995
    linearSpringConstant=50
    simulationSpace=BoneRelative
    alphaBoolBlend {blendOption=Linear}
    boundBone {boneName=dyn_cube_r_in_02}
    componentAppliedLinearAccClamp=<6000 6000 6000>
    componentLinearAccScale=<0.3 0.3 0.1>
    componentLinearVelScale=<0.7 0.7 0.8>
    gravityOverride {z=-980}
    physicsBodyDefinitions [{
      sphereCollisionRadius=0
      boundBone {boneName=dyn_cube_r_in_02}
      boxExtents=<5 5 5>
      constraintSetup {
       angularTargetAxis=AxisZ
       twistAxis=AxisZ
       angularLimitsMax=<0 0 0>
       angularLimitsMin=<0 0 0>
       angularTarget=<0 0 0>
       linearAxesMax=<18 18 18>
       linearAxesMin=<-0.1 -0.1 -2>
      }
      localJointOffset=<-0.1 -0.2 0>
     }]
    relativeSpaceBone {boneName=dyn_master}
   }
   n7 = Anim Dynamics - Bone: dyn_fin_armA_01_l(Alpha=1, ExternalForce="0,0,0") {
    angularDampingOverride=0.9999
    angularSpringConstant=500
    bAngularSpring=true
    bGravityOverrideInSimSpace=true
    bOverrideAngularDamping=true
    simulationSpace=BoneRelative
    alphaBoolBlend {blendOption=Linear}
    boundBone {boneName=dyn_fin_armA_01_l}
    componentAppliedLinearAccClamp=<3500 3500 3500>
    componentLinearAccScale=<1.2 1.2 1.2>
    componentLinearVelScale=<1.2 1.2 1.2>
    physicsBodyDefinitions [{
      sphereCollisionRadius=0
      boundBone {boneName=dyn_fin_armA_01_l}
      boxExtents=<15 10 10>
      constraintSetup {angularLimitsMax=<1 5 5>, angularLimitsMin=<-1 -20 -15>, angularTarget=<0 0 0>}
      localJointOffset=<5 0 0>
     }]
    relativeSpaceBone {boneName=spine_05}
   }
   n8 = Anim Dynamics - Bone: dyn_fin_armB_01_l(ComponentPose=@n7, Alpha=1, ExternalForce="0,0,0") {
    angularDampingOverride=0.9999
    angularSpringConstant=500
    bAngularSpring=true
    bGravityOverrideInSimSpace=true
    bOverrideAngularDamping=true
    simulationSpace=BoneRelative
    alphaBoolBlend {blendOption=Linear}
    boundBone {boneName=dyn_fin_armB_01_l}
    componentAppliedLinearAccClamp=<3500 3500 3500>
    componentLinearAccScale=<1.2 1.2 1.2>
    componentLinearVelScale=<1.2 1.2 1.2>
    physicsBodyDefinitions [{
      sphereCollisionRadius=0
      boundBone {boneName=dyn_fin_armB_01_l}
      boxExtents=<15 10 10>
      constraintSetup {angularLimitsMax=<1 5 5>, angularLimitsMin=<-1 -20 -15>, angularTarget=<0 0 0>}
      localJointOffset=<5 0 0>
     }]
    relativeSpaceBone {boneName=spine_05}
   }
   n9 = Anim Dynamics - Bone: dyn_fin_armC_01_l(ComponentPose=@n8, Alpha=1, ExternalForce="0,0,0") {
    angularDampingOverride=0.9999
    angularSpringConstant=500
    bAngularSpring=true
    bGravityOverrideInSimSpace=true
    bOverrideAngularDamping=true
    simulationSpace=BoneRelative
    alphaBoolBlend {blendOption=Linear}
    boundBone {boneName=dyn_fin_armC_01_l}
    componentAppliedLinearAccClamp=<3500 3500 3500>
    componentLinearAccScale=<1.2 1.2 1.2>
    componentLinearVelScale=<1.2 1.2 1.2>
    physicsBodyDefinitions [{
      sphereCollisionRadius=0
      boundBone {boneName=dyn_fin_armC_01_l}
      boxExtents=<15 10 10>
      constraintSetup {angularLimitsMax=<1 5 5>, angularLimitsMin=<-1 -20 -15>, angularTarget=<0 0 0>}
      localJointOffset=<5 0 0>
     }]
    relativeSpaceBone {boneName=spine_05}
   }
   n10 = Anim Dynamics - Bone: dyn_fin_armA_01_r(ComponentPose=@n9, Alpha=1, ExternalForce="0,0,0") {
    angularDampingOverride=0.9999
    angularSpringConstant=500
    bAngularSpring=true
    bGravityOverrideInSimSpace=true
    bOverrideAngularDamping=true
    simulationSpace=BoneRelative
    alphaBoolBlend {blendOption=Linear}
    boundBone {boneName=dyn_fin_armA_01_r}
    componentAppliedLinearAccClamp=<3500 3500 3500>
    componentLinearAccScale=<1.2 1.2 1.2>
    componentLinearVelScale=<1.2 1.2 1.2>
    physicsBodyDefinitions [{
      sphereCollisionRadius=0
      boundBone {boneName=dyn_fin_armA_01_r}
      boxExtents=<15 10 10>
      constraintSetup {angularLimitsMax=<1 1 5>, angularLimitsMin=<-5 -35 -15>, angularTarget=<0 0 0>}
      localJointOffset=<5 0 0>
     }]
    relativeSpaceBone {boneName=spine_05}
   }
   n11 = Anim Dynamics - Bone: dyn_fin_armB_01_r(ComponentPose=@n10, Alpha=1, ExternalForce="0,0,0") {
    angularDampingOverride=0.9999
    angularSpringConstant=500
    bAngularSpring=true
    bGravityOverrideInSimSpace=true
    bOverrideAngularDamping=true
    simulationSpace=BoneRelative
    alphaBoolBlend {blendOption=Linear}
    boundBone {boneName=dyn_fin_armB_01_r}
    componentAppliedLinearAccClamp=<3500 3500 3500>
    componentLinearAccScale=<1.2 1.2 1.2>
    componentLinearVelScale=<1.2 1.2 1.2>
    physicsBodyDefinitions [{
      sphereCollisionRadius=0
      boundBone {boneName=dyn_fin_armB_01_r}
      boxExtents=<15 10 10>
      constraintSetup {angularLimitsMax=<1 1 5>, angularLimitsMin=<-5 -35 -15>, angularTarget=<0 0 0>}
      localJointOffset=<5 0 0>
     }]
    relativeSpaceBone {boneName=spine_05}
   }
   Anim Dynamics - Bone: dyn_fin_armC_01_r(ComponentPose=@n11, Alpha=1, ExternalForce="0,0,0") {
    angularDampingOverride=0.9999
    angularSpringConstant=500
    bAngularSpring=true
    bGravityOverrideInSimSpace=true
    bOverrideAngularDamping=true
    simulationSpace=BoneRelative
    alphaBoolBlend {blendOption=Linear}
    boundBone {boneName=dyn_fin_armC_01_r}
    componentAppliedLinearAccClamp=<3500 3500 3500>
    componentLinearAccScale=<1.2 1.2 1.2>
    componentLinearVelScale=<1.2 1.2 1.2>
    physicsBodyDefinitions [{
      sphereCollisionRadius=0
      boundBone {boneName=dyn_fin_armC_01_r}
      boxExtents=<15 10 10>
      constraintSetup {angularLimitsMax=<1 1 5>, angularLimitsMin=<-5 -35 -15>, angularTarget=<0 0 0>}
      localJointOffset=<5 0 0>
     }]
    relativeSpaceBone {boneName=spine_05}
   }
   "Transform (Modify) Bone - Bone: dyn_cube_l_out_04"(Alpha=0.8, Translation="0,0,0", Rotation="0,0,0", Scale="1,1,1") {
    rotationMode=BMM_Additive
    rotationSpace=BCS_BoneSpace
    scaleMode=BMM_Additive
    scaleSpace=BCS_BoneSpace
    translationMode=BMM_Additive
    translationSpace=BCS_BoneSpace
    alphaBoolBlend {blendOption=Linear}
    boneToModify {boneName=dyn_cube_l_out_04}
   }
   "Transform (Modify) Bone - Bone: dyn_skirt_hip_01_l"(Alpha=1, Translation="0,0,0", Rotation="-25,0.000,0.000", Scale="1,1,1") {
    rotationMode=BMM_Additive
    rotationSpace=BCS_BoneSpace
    alphaBoolBlend {blendOption=Linear}
    alphaScaleBiasClamp {bInterpResult=true, interpSpeedDecreasing=15, interpSpeedIncreasing=15}
    boneToModify {boneName=dyn_skirt_hip_01_l}
    rotation {pitch=-25}
   }
   RigidBody() {
    bClampLinearTranslationLimitToRefPose=true
    bOverrideWorldGravity=true
    overridePhysicsAsset=&/F_MED_Lady_Atlantis_Physics
    simulationSpace=BaseBoneSpace
    alphaBoolBlend {blendOption=Linear}
    baseBoneRef {boneName=root}
    componentAppliedLinearAccClamp=<600 600 800>
    componentLinearAccScale=<0.5 0.5 0.2>
    componentLinearVelScale=<0.2 0.2 0.2>
    overrideWorldGravity {z=-680}
   }
   n12 = Get thigh_rot_r()
   n13 = RigidBody(ComponentPose=Local To Component(LocalPose=Use cached pose 'Pose'())) {
    bClampLinearTranslationLimitToRefPose=true
    bOverrideWorldGravity=true
    overridePhysicsAsset=&/F_MED_Lady_Atlantis_Dress_Capsules
    simulationSpace=BaseBoneSpace
    alphaBoolBlend {blendOption=Linear}
    baseBoneRef {boneName=pelvis}
    componentAppliedLinearAccClamp=<600 600 600>
    componentLinearAccScale=<0.5 0.5 0.2>
    componentLinearVelScale=<0.2 0.2 0.2>
    overrideWorldGravity {z=-450}
   }
   n14 = "Constraint - dyn_skirt_side_02_l (P)"(ComponentPose=@n13, Alpha=0.8, ConstraintWeights_0=1) {
    alphaBoolBlend {blendOption=Linear}
    alphaScaleBiasClamp {clampMax=0.35}
    boneToModify {boneName=dyn_skirt_side_02_l}
    constraintSetup [{targetBone {boneName=thigh_l}}]
    constraintWeights [1]
   }
   n15 = "Constraint - dyn_skirt_bck_03_l (P)"(ComponentPose=@n14, Alpha=0.5, ConstraintWeights_0=1) {
    alphaBoolBlend {blendOption=Linear}
    alphaScaleBiasClamp {clampMax=0.35}
    boneToModify {boneName=dyn_skirt_bck_03_l}
    constraintSetup [{targetBone {boneName=thigh_l}}]
    constraintWeights [1]
   }
   n16 = "Constraint - dyn_skirt_side_02_r (P)"(ComponentPose=@n15, Alpha=0.8, ConstraintWeights_0=1) {
    alphaBoolBlend {blendOption=Linear}
    alphaScaleBiasClamp {clampMax=0.35}
    boneToModify {boneName=dyn_skirt_side_02_r}
    constraintSetup [{targetBone {boneName=thigh_r}}]
    constraintWeights [1]
   }
   n17 = "Constraint - dyn_skirt_bck_03_r (P)"(ComponentPose=@n16, Alpha=0.5, ConstraintWeights_0=1) {
    alphaBoolBlend {blendOption=Linear}
    alphaScaleBiasClamp {clampMax=0.35}
    boneToModify {boneName=dyn_skirt_bck_03_r}
    constraintSetup [{targetBone {boneName=thigh_r}}]
    constraintWeights [1]
   }
   n18 = "Constraint - dyn_skirt_bckA_03_l (P)"(ComponentPose=@n17, Alpha=0.5, ConstraintWeights_0=1) {
    alphaBoolBlend {blendOption=Linear}
    alphaScaleBiasClamp {clampMax=0.35}
    boneToModify {boneName=dyn_skirt_bckA_03_l}
    constraintSetup [{targetBone {boneName=thigh_l}}]
    constraintWeights [1]
   }
   n19 = "Constraint - dyn_skirt_bckA_03_r (P)"(ComponentPose=@n18, Alpha=0.5, ConstraintWeights_0=1) {
    alphaBoolBlend {blendOption=Linear}
    alphaScaleBiasClamp {clampMax=0.35}
    boneToModify {boneName=dyn_skirt_bckA_03_r}
    constraintSetup [{targetBone {boneName=thigh_r}}]
    constraintWeights [1]
   }
   n20 = "Constraint - dyn_skirt_bckA_04_l (P)"(ComponentPose=@n19, Alpha=0.5, ConstraintWeights_0=1) {
    alphaBoolBlend {blendOption=Linear}
    alphaScaleBiasClamp {clampMax=0.35}
    boneToModify {boneName=dyn_skirt_bckA_04_l}
    constraintSetup [{targetBone {boneName=calf_l}}]
    constraintWeights [1]
   }
   n21 = "Constraint - dyn_skirt_bckA_04_r (P)"(ComponentPose=@n20, Alpha=0.5, ConstraintWeights_0=1) {
    alphaBoolBlend {blendOption=Linear}
    alphaScaleBiasClamp {clampMax=0.35}
    boneToModify {boneName=dyn_skirt_bckA_04_r}
    constraintSetup [{targetBone {boneName=calf_r}}]
    constraintWeights [1]
   }
   n22 = "Transform (Modify) Bone - Bone: dyn_skirt_side_01_l"(ComponentPose=@n21, Alpha=1, Translation="0,-2,0", Rotation="10.000,0.000,0.000", Scale="1,1,1") {
    rotationMode=BMM_Additive
    rotationSpace=BCS_BoneSpace
    alphaBoolBlend {blendOption=Linear}
    alphaScaleBiasClamp {bInterpResult=true, interpSpeedDecreasing=15, interpSpeedIncreasing=15}
    boneToModify {boneName=dyn_skirt_side_01_l}
    rotation {pitch=10}
    translation {y=-2}
   }
   n23 = "Transform (Modify) Bone - Bone: dyn_skirt_side_01_r"(ComponentPose=@n22, Alpha=1, Translation="0,-2,0", Rotation="10.000,0.000,0.000", Scale="1,1,1") {
    rotationMode=BMM_Additive
    rotationSpace=BCS_BoneSpace
    translationMode=BMM_Additive
    alphaBoolBlend {blendOption=Linear}
    alphaScaleBiasClamp {bInterpResult=true, interpSpeedDecreasing=15, interpSpeedIncreasing=15}
    boneToModify {boneName=dyn_skirt_side_01_r}
    rotation {pitch=10}
    translation {y=-2}
   }
   n24 = "Transform (Modify) Bone - Bone: dyn_skirt_bck_01_l"(ComponentPose=@n23, Alpha=1, Translation="0,-2,0", Rotation="10.000,0.000,0.000", Scale="1,1,1") {
    rotationMode=BMM_Additive
    rotationSpace=BCS_BoneSpace
    alphaBoolBlend {blendOption=Linear}
    alphaScaleBiasClamp {bInterpResult=true, interpSpeedDecreasing=15, interpSpeedIncreasing=15}
    boneToModify {boneName=dyn_skirt_bck_01_l}
    rotation {pitch=10}
    translation {y=-2}
   }
   n25 = "Transform (Modify) Bone - Bone: dyn_skirt_bck_01_r"(ComponentPose=@n24, Alpha=1, Translation="0,2,0", Rotation="10.000,0.000,0.000", Scale="1,1,1") {
    rotationMode=BMM_Additive
    rotationSpace=BCS_BoneSpace
    translationMode=BMM_Additive
    alphaBoolBlend {blendOption=Linear}
    alphaScaleBiasClamp {bInterpResult=true, interpSpeedDecreasing=15, interpSpeedIncreasing=15}
    boneToModify {boneName=dyn_skirt_bck_01_r}
    rotation {pitch=10}
    translation {y=2}
   }
   n26 = "Transform (Modify) Bone - Bone: dyn_skirt_bckA_01_l"(ComponentPose=@n25, Alpha=1, Translation="0,2,0", Rotation="10,0.000,0.000", Scale="1,1,1") {
    rotationMode=BMM_Additive
    rotationSpace=BCS_BoneSpace
    translationMode=BMM_Additive
    alphaBoolBlend {blendOption=Linear}
    alphaScaleBiasClamp {bInterpResult=true, interpSpeedDecreasing=15, interpSpeedIncreasing=15}
    boneToModify {boneName=dyn_skirt_bckA_01_l}
    rotation {pitch=10}
    translation {y=2}
   }
   n27 = "Transform (Modify) Bone - Bone: dyn_skirt_bckA_01_r"(ComponentPose=@n26, Alpha=1, Translation="0,2,0", Rotation="10,0.000,0.000", Scale="1,1,1") {
    rotationMode=BMM_Additive
    rotationSpace=BCS_BoneSpace
    translationMode=BMM_Additive
    alphaBoolBlend {blendOption=Linear}
    alphaScaleBiasClamp {bInterpResult=true, interpSpeedDecreasing=15, interpSpeedIncreasing=15}
    boneToModify {boneName=dyn_skirt_bckA_01_r}
    rotation {pitch=10}
    translation {y=2}
   }
   n28 = Component To Local(ComponentPose=@n27)
   n29 = RigidBody(ComponentPose=Local To Component(LocalPose=Use cached pose 'Pose'())) {
    bClampLinearTranslationLimitToRefPose=true
    bOverrideWorldGravity=true
    overridePhysicsAsset=&F_MED_Lady_Atlantis_Dress_Capsules
    simulationSpace=BaseBoneSpace
    alphaBoolBlend {blendOption=Linear}
    baseBoneRef {boneName=root}
    componentAppliedLinearAccClamp=<600 600 600>
    componentLinearAccScale=<0.5 0.5 0.2>
    componentLinearVelScale=<0.2 0.2 0.2>
    overrideWorldGravity {z=-450}
   }
   n30 = Component To Local(ComponentPose=@n29)
   n31 = Get bIsSkyDiving()
   n32 = Anim Dynamics - Bone: dyn_fin_armA_01_l(
    ComponentPose=Local To Component(LocalPose=Blend Poses by bool(BlendPose_0=@n28, BlendPose_1=@n30, bActiveValue=@n31) {blendType=Linear})
    Alpha=0.3
    ExternalForce="0,0,0"
   ) {
    bGravityOverrideInSimSpace=true
    bUseGravityOverride=true
    gravityScale=0.3
    lODThreshold=1
    simulationSpace=BoneRelative
    alphaBoolBlend {blendOption=Linear}
    alphaScaleBiasClamp {clampMax=0.09, clampMin=0.05}
    boundBone {boneName=dyn_fin_armA_01_l}
    gravityOverride {z=-980}
    physicsBodyDefinitions [{
      sphereCollisionRadius=0
      boundBone {boneName=dyn_fin_armA_01_l}
      boxExtents=<2 5 5>
      constraintSetup {
       angularLimitsMax=<1.5 1.5 5>
       angularLimitsMin=<-1.5 -1.5 -20>
       angularTarget=<0 0 0>
       linearAxesMax=<0.5 1 0.5>
       linearAxesMin=<-1.5 -5 -1.5>
      }
      localJointOffset=<5 0 0>
     }]
    relativeSpaceBone {boneName=pelvis}
   }
   n33 = Anim Dynamics - Bone: dyn_fin_armB_01_l(ComponentPose=@n32, Alpha=0.3, ExternalForce="0,0,0") {
    bGravityOverrideInSimSpace=true
    bUseGravityOverride=true
    gravityScale=0.3
    lODThreshold=1
    simulationSpace=BoneRelative
    alphaBoolBlend {blendOption=Linear}
    alphaScaleBiasClamp {clampMax=0.09, clampMin=0.05}
    boundBone {boneName=dyn_fin_armB_01_l}
    gravityOverride {z=-980}
    physicsBodyDefinitions [{
      sphereCollisionRadius=0
      boundBone {boneName=dyn_fin_armB_01_l}
      boxExtents=<2 5 5>
      constraintSetup {
       angularLimitsMax=<1.5 1.5 5>
       angularLimitsMin=<-1.5 -1.5 -20>
       angularTarget=<0 0 0>
       linearAxesMax=<0.5 1 0.5>
       linearAxesMin=<-1.5 -5 -1.5>
      }
      localJointOffset=<5 0 0>
     }]
    relativeSpaceBone {boneName=pelvis}
   }
   n34 = Anim Dynamics - Bone: dyn_fin_armC_01_l(ComponentPose=@n33, Alpha=0.3, ExternalForce="0,0,0") {
    bGravityOverrideInSimSpace=true
    bUseGravityOverride=true
    gravityScale=0.3
    lODThreshold=1
    simulationSpace=BoneRelative
    alphaBoolBlend {blendOption=Linear}
    alphaScaleBiasClamp {clampMax=0.09, clampMin=0.05}
    boundBone {boneName=dyn_fin_armC_01_l}
    gravityOverride {z=-980}
    physicsBodyDefinitions [{
      sphereCollisionRadius=0
      boundBone {boneName=dyn_fin_armC_01_l}
      boxExtents=<2 5 5>
      constraintSetup {
       angularLimitsMax=<1.5 1.5 5>
       angularLimitsMin=<-1.5 -1.5 -20>
       angularTarget=<0 0 0>
       linearAxesMax=<0.5 1 0.5>
       linearAxesMin=<-1.5 -5 -1.5>
      }
      localJointOffset=<5 0 0>
     }]
    relativeSpaceBone {boneName=pelvis}
   }
   n35 = Anim Dynamics - Bone: dyn_fin_armA_01_r(ComponentPose=@n34, Alpha=0.3, ExternalForce="0,0,0") {
    bGravityOverrideInSimSpace=true
    bUseGravityOverride=true
    gravityScale=0.3
    lODThreshold=1
    simulationSpace=BoneRelative
    alphaBoolBlend {blendOption=Linear}
    alphaScaleBiasClamp {clampMax=0.09, clampMin=0.05}
    boundBone {boneName=dyn_fin_armA_01_r}
    gravityOverride {z=-980}
    physicsBodyDefinitions [{
      sphereCollisionRadius=0
      boundBone {boneName=dyn_fin_armA_01_r}
      boxExtents=<2 5 5>
      constraintSetup {
       angularLimitsMax=<1.5 1.5 5>
       angularLimitsMin=<-1.5 -1.5 -20>
       angularTarget=<0 0 0>
       linearAxesMax=<0.5 1 0.5>
       linearAxesMin=<-1.5 -5 -1.5>
      }
      localJointOffset=<5 0 0>
     }]
    relativeSpaceBone {boneName=pelvis}
   }
   n36 = Anim Dynamics - Bone: dyn_fin_armB_01_r(ComponentPose=@n35, Alpha=0.3, ExternalForce="0,0,0") {
    bGravityOverrideInSimSpace=true
    bUseGravityOverride=true
    gravityScale=0.3
    lODThreshold=1
    simulationSpace=BoneRelative
    alphaBoolBlend {blendOption=Linear}
    alphaScaleBiasClamp {clampMax=0.09, clampMin=0.05}
    boundBone {boneName=dyn_fin_armB_01_r}
    gravityOverride {z=-980}
    physicsBodyDefinitions [{
      sphereCollisionRadius=0
      boundBone {boneName=dyn_fin_armB_01_r}
      boxExtents=<2 5 5>
      constraintSetup {
       angularLimitsMax=<1.5 1.5 5>
       angularLimitsMin=<-1.5 -1.5 -20>
       angularTarget=<0 0 0>
       linearAxesMax=<0.5 1 0.5>
       linearAxesMin=<-1.5 -5 -1.5>
      }
      localJointOffset=<5 0 0>
     }]
    relativeSpaceBone {boneName=pelvis}
   }
   n37 = Anim Dynamics - Bone: dyn_fin_armC_01_r(ComponentPose=@n36, Alpha=0.3, ExternalForce="0,0,0") {
    bGravityOverrideInSimSpace=true
    bUseGravityOverride=true
    gravityScale=0.3
    lODThreshold=1
    simulationSpace=BoneRelative
    alphaBoolBlend {blendOption=Linear}
    alphaScaleBiasClamp {clampMax=0.09, clampMin=0.05}
    boundBone {boneName=dyn_fin_armC_01_r}
    gravityOverride {z=-980}
    physicsBodyDefinitions [{
      sphereCollisionRadius=0
      boundBone {boneName=dyn_fin_armC_01_r}
      boxExtents=<2 5 5>
      constraintSetup {
       angularLimitsMax=<1.5 1.5 5>
       angularLimitsMin=<-1.5 -1.5 -20>
       angularTarget=<0 0 0>
       linearAxesMax=<0.5 1 0.5>
       linearAxesMin=<-1.5 -5 -1.5>
      }
      localJointOffset=<5 0 0>
     }]
    relativeSpaceBone {boneName=pelvis}
   }
   n38 = Anim Dynamics - Bone: dyn_fin_calfA_01_l(ComponentPose=@n37, Alpha=1, ExternalForce="0,0,0") {
    angularDampingOverride=0.9999
    angularSpringConstant=500
    bAngularSpring=true
    bGravityOverrideInSimSpace=true
    bOverrideAngularDamping=true
    simulationSpace=BoneRelative
    alphaBoolBlend {blendOption=Linear}
    boundBone {boneName=dyn_fin_calfA_01_l}
    componentAppliedLinearAccClamp=<3500 3500 3500>
    componentLinearAccScale=<1.2 1.2 1.2>
    componentLinearVelScale=<1.2 1.2 1.2>
    physicsBodyDefinitions [{
      sphereCollisionRadius=0
      boundBone {boneName=dyn_fin_calfA_01_l}
      boxExtents=<15 10 10>
      constraintSetup {angularLimitsMax=<5 5 2>, angularLimitsMin=<-20 -35 -15>, angularTarget=<0 0 0>, linearAxesMax=<1.5 1.5 1.5>, linearAxesMin=<-1.5 -5 -1.5>}
      localJointOffset=<5 0 0>
     }]
    relativeSpaceBone {boneName=pelvis}
   }
   n39 = Anim Dynamics - Bone: dyn_fin_calfB_01_l(ComponentPose=@n38, Alpha=1, ExternalForce="0,0,0") {
    angularDampingOverride=0.9999
    angularSpringConstant=500
    bAngularSpring=true
    bGravityOverrideInSimSpace=true
    bOverrideAngularDamping=true
    simulationSpace=BoneRelative
    alphaBoolBlend {blendOption=Linear}
    boundBone {boneName=dyn_fin_calfB_01_l}
    componentAppliedLinearAccClamp=<3500 3500 3500>
    componentLinearAccScale=<1.2 1.2 1.2>
    componentLinearVelScale=<1.2 1.2 1.2>
    physicsBodyDefinitions [{
      sphereCollisionRadius=0
      boundBone {boneName=dyn_fin_calfB_01_l}
      boxExtents=<15 10 10>
      constraintSetup {angularLimitsMax=<5 5 2>, angularLimitsMin=<-20 -35 -15>, angularTarget=<0 0 0>, linearAxesMax=<1.5 1.5 1.5>, linearAxesMin=<-1.5 -5 -1.5>}
      localJointOffset=<5 0 0>
     }]
    relativeSpaceBone {boneName=pelvis}
   }
   n40 = Anim Dynamics - Bone: dyn_fin_calfC_01_l(ComponentPose=@n39, Alpha=1, ExternalForce="0,0,0") {
    angularDampingOverride=0.9999
    angularSpringConstant=500
    bAngularSpring=true
    bGravityOverrideInSimSpace=true
    bOverrideAngularDamping=true
    simulationSpace=BoneRelative
    alphaBoolBlend {blendOption=Linear}
    boundBone {boneName=dyn_fin_calfC_01_l}
    componentAppliedLinearAccClamp=<3500 3500 3500>
    componentLinearAccScale=<1.2 1.2 1.2>
    componentLinearVelScale=<1.2 1.2 1.2>
    physicsBodyDefinitions [{
      sphereCollisionRadius=0
      boundBone {boneName=dyn_fin_calfC_01_l}
      boxExtents=<15 10 10>
      constraintSetup {angularLimitsMax=<5 5 2>, angularLimitsMin=<-20 -35 -15>, angularTarget=<0 0 0>, linearAxesMax=<1.5 1.5 1.5>, linearAxesMin=<-1.5 -5 -1.5>}
      localJointOffset=<5 0 0>
     }]
    relativeSpaceBone {boneName=pelvis}
   }
   n41 = Anim Dynamics - Bone: dyn_fin_calfA_01_r(ComponentPose=@n40, Alpha=1, ExternalForce="0,0,0") {
    angularDampingOverride=0.9999
    angularSpringConstant=500
    bAngularSpring=true
    bGravityOverrideInSimSpace=true
    bOverrideAngularDamping=true
    simulationSpace=BoneRelative
    alphaBoolBlend {blendOption=Linear}
    boundBone {boneName=dyn_fin_calfA_01_r}
    componentAppliedLinearAccClamp=<3500 3500 3500>
    componentLinearAccScale=<1.2 1.2 1.2>
    componentLinearVelScale=<1.2 1.2 1.2>
    physicsBodyDefinitions [{
      sphereCollisionRadius=0
      boundBone {boneName=dyn_fin_calfA_01_r}
      boxExtents=<15 10 10>
      constraintSetup {angularLimitsMax=<5 5 2>, angularLimitsMin=<-20 -35 -15>, angularTarget=<0 0 0>, linearAxesMax=<1.5 1.5 1.5>, linearAxesMin=<-1.5 -5 -1.5>}
      localJointOffset=<5 0 0>
     }]
    relativeSpaceBone {boneName=pelvis}
   }
   n42 = Anim Dynamics - Bone: dyn_fin_calfB_01_r(ComponentPose=@n41, Alpha=1, ExternalForce="0,0,0") {
    angularDampingOverride=0.9999
    angularSpringConstant=500
    bAngularSpring=true
    bGravityOverrideInSimSpace=true
    bOverrideAngularDamping=true
    simulationSpace=BoneRelative
    alphaBoolBlend {blendOption=Linear}
    boundBone {boneName=dyn_fin_calfB_01_r}
    componentAppliedLinearAccClamp=<3500 3500 3500>
    componentLinearAccScale=<1.2 1.2 1.2>
    componentLinearVelScale=<1.2 1.2 1.2>
    physicsBodyDefinitions [{
      sphereCollisionRadius=0
      boundBone {boneName=dyn_fin_calfB_01_r}
      boxExtents=<15 10 10>
      constraintSetup {angularLimitsMax=<5 5 2>, angularLimitsMin=<-20 -35 -15>, angularTarget=<0 0 0>, linearAxesMax=<1.5 1.5 1.5>, linearAxesMin=<-1.5 -5 -1.5>}
      localJointOffset=<5 0 0>
     }]
    relativeSpaceBone {boneName=pelvis}
   }
   n43 = Anim Dynamics - Bone: dyn_fin_calfC_01_r(ComponentPose=@n42, Alpha=1, ExternalForce="0,0,0") {
    angularDampingOverride=0.9999
    angularSpringConstant=500
    bAngularSpring=true
    bGravityOverrideInSimSpace=true
    bOverrideAngularDamping=true
    simulationSpace=BoneRelative
    alphaBoolBlend {blendOption=Linear}
    boundBone {boneName=dyn_fin_calfC_01_r}
    componentAppliedLinearAccClamp=<3500 3500 3500>
    componentLinearAccScale=<1.2 1.2 1.2>
    componentLinearVelScale=<1.2 1.2 1.2>
    physicsBodyDefinitions [{
      sphereCollisionRadius=0
      boundBone {boneName=dyn_fin_calfC_01_r}
      boxExtents=<15 10 10>
      constraintSetup {angularLimitsMax=<5 5 2>, angularLimitsMin=<-20 -35 -15>, angularTarget=<0 0 0>, linearAxesMax=<1.5 1.5 1.5>, linearAxesMin=<-1.5 -5 -1.5>}
      localJointOffset=<5 0 0>
     }]
    relativeSpaceBone {boneName=pelvis}
   }
   n44 = Get thigh_rot_l()
   n45 = "Transform (Modify) Bone - Bone: dyn_skirt_hip_02_l"(ComponentPose=@n43, Alpha=@n44, Translation="0,0,0", Rotation="15,25,-0.000", Scale="1,1,1") {
    rotationMode=BMM_Additive
    rotationSpace=BCS_BoneSpace
    alphaBoolBlend {blendOption=Linear}
    alphaScaleBiasClamp {bInterpResult=true, interpSpeedDecreasing=15, interpSpeedIncreasing=15}
    boneToModify {boneName=dyn_skirt_hip_02_l}
    rotation {pitch=15, yaw=25}
   }
   n46 = "Transform (Modify) Bone - Bone: dyn_skirt_hip_03_l"(ComponentPose=@n45, Alpha=@n44, Translation="0,0,0", Rotation="0,65,-0.000", Scale="1,1,1") {
    rotationMode=BMM_Additive
    rotationSpace=BCS_BoneSpace
    alphaBoolBlend {blendOption=Linear}
    alphaScaleBiasClamp {bInterpResult=true, interpSpeedDecreasing=15, interpSpeedIncreasing=15}
    boneToModify {boneName=dyn_skirt_hip_03_l}
    rotation {yaw=65}
   }
   n47 = "Transform (Modify) Bone - Bone: dyn_skirt_hip_02_r"(ComponentPose=@n46, Alpha=@n12, Translation="0,0.5,0", Rotation="15,25,-0.000", Scale="1,1,1") {
    rotationMode=BMM_Additive
    rotationSpace=BCS_BoneSpace
    translationMode=BMM_Additive
    alphaBoolBlend {blendOption=Linear}
    alphaScaleBiasClamp {bInterpResult=true, interpSpeedDecreasing=15, interpSpeedIncreasing=15}
    boneToModify {boneName=dyn_skirt_hip_02_r}
    rotation {pitch=15, yaw=25}
    translation {y=0.5}
   }
   n48 = "Transform (Modify) Bone - Bone: dyn_skirt_hip_03_r"(ComponentPose=@n47, Alpha=@n12, Translation="0,1,0", Rotation="0,65,-0.000", Scale="1,1,1") {
    rotationMode=BMM_Additive
    rotationSpace=BCS_BoneSpace
    translationMode=BMM_Additive
    alphaBoolBlend {blendOption=Linear}
    alphaScaleBiasClamp {bInterpResult=true, interpSpeedDecreasing=15, interpSpeedIncreasing=15}
    boneToModify {boneName=dyn_skirt_hip_03_r}
    rotation {yaw=65}
    translation {y=1}
   }
   n49 = "Transform (Modify) Bone - Bone: dyn_skirt_hip_01_r"(ComponentPose=@n48, Alpha=@n12, Translation="0,0,0", Rotation="0.000,-20.000,0.000", Scale="1,1,1") {
    alpha=0
    rotationMode=BMM_Additive
    rotationSpace=BCS_BoneSpace
    alphaBoolBlend {blendOption=Linear}
    alphaScaleBiasClamp {bInterpResult=true, interpSpeedDecreasing=15, interpSpeedIncreasing=15}
    boneToModify {boneName=dyn_skirt_hip_01_r}
    rotation {yaw=-20}
   }
   Output Pose(Result=Component To Local(ComponentPose=@n49)) {layerGroup=None}
   n50 = Get thigh_rot_l()
   n51 = "Transform (Modify) Bone - Bone: dyn_driver_skirt_hip_01_l"(
    ComponentPose=Local To Component(LocalPose=Copy Pose From Mesh() {bUseAttachedParent=true})
    Alpha=@n50
    Translation="0.852,-1.828,-1.652"
    Rotation="5,-25,-0.000"
    Scale="1,1,1"
   ) {
    rotationMode=BMM_Additive
    rotationSpace=BCS_BoneSpace
    translationMode=BMM_Additive
    translationSpace=BCS_BoneSpace
    alphaBoolBlend {blendOption=Linear}
    alphaScaleBiasClamp {bInterpResult=true, interpSpeedDecreasing=15, interpSpeedIncreasing=15}
    boneToModify {boneName=dyn_driver_skirt_hip_01_l}
    rotation {pitch=5, yaw=-25}
    translation=<0.852 -1.828 -1.652>
   }
   n52 = "Transform (Modify) Bone - Bone: dyn_driver_skirt_side_01_l"(ComponentPose=@n51, Alpha=@n50, Translation="0,0,0", Rotation="10.000,0.000,0.000", Scale="1,1,1") {
    rotationMode=BMM_Additive
    rotationSpace=BCS_BoneSpace
    translationMode=BMM_Additive
    translationSpace=BCS_BoneSpace
    alphaBoolBlend {blendOption=Linear}
    alphaScaleBiasClamp {bInterpResult=true, interpSpeedDecreasing=15, interpSpeedIncreasing=15}
    boneToModify {boneName=dyn_driver_skirt_side_01_l}
    rotation {pitch=10}
   }
   n53 = Get thigh_rot_r()
   n54 = "Transform (Modify) Bone - Bone: dyn_parent_skirt_hip_01_r"(ComponentPose=@n52, Alpha=@n53, Translation="0,0,0", Rotation="5.000,-50.000,0.000", Scale="1,1,1") {
    rotationMode=BMM_Additive
    rotationSpace=BCS_BoneSpace
    translationMode=BMM_Additive
    translationSpace=BCS_BoneSpace
    alphaBoolBlend {blendOption=Linear}
    alphaScaleBiasClamp {bInterpResult=true, interpSpeedDecreasing=15, interpSpeedIncreasing=15}
    boneToModify {boneName=dyn_parent_skirt_hip_01_r}
    rotation {pitch=5, yaw=-50}
   }
   n55 = "Transform (Modify) Bone - Bone: dyn_driver_skirt_side_01_r"(ComponentPose=@n54, Alpha=@n53, Translation="0,0,0", Rotation="10.000,0.000,0.000", Scale="1,1,1") {
    rotationMode=BMM_Additive
    rotationSpace=BCS_BoneSpace
    translationMode=BMM_Additive
    translationSpace=BCS_BoneSpace
    alphaBoolBlend {blendOption=Linear}
    alphaScaleBiasClamp {bInterpResult=true, interpSpeedDecreasing=15, interpSpeedIncreasing=15}
    boneToModify {boneName=dyn_driver_skirt_side_01_r}
    rotation {pitch=10}
   }
   Save cached pose 'Pose'(Pose=Component To Local(ComponentPose=@n55)) {cacheName=Pose}
  ]}
 EventGraph {nodes [
   Event BlueprintUpdateAnimation()
   > Get!(execute:Exec) -> (then:Exec, else:Exec, Skel_Mesh:Skeletal Mesh Component Object Reference)
   > n1 = Sequence!(execute) -> (then_0, then_1, then_2:Exec, then_3:Exec)
   > Set thigh_rot_r(thigh_rot_r=GetAlphaFromReferenceTransformDelta(SkeletalMeshComponent=Get Skel_Mesh(), BoneName=thigh_r, BaseName=pelvis, InputRange=(X=0,Y=-90), TwistAxis="0,0,1"))
   Set thigh_rot_l(
    execute=@n1.then_1
    thigh_rot_l=GetAlphaFromReferenceTransformDelta(SkeletalMeshComponent=Get Skel_Mesh(), BoneName=thigh_l, BaseName=pelvis, InputRange=(X=-20,Y=-101), TwistAxis="0,0,1")
   )
   n2 = Branch(execute=@n1.then_2, Condition=Get bIsSkyDiving())
   > n3 = Branch(Condition=Get bIsParachuteOpened())
   > Set gravity_override(gravity_override="0, 0,-800")
   n4 = Break Fort Anim Input Skydiving External Force(FortAnimInput_SkydivingExternalForce=Get SkydivingExternalForceInput()) -> (
    bUseSkydivingVectorForce=~
    SkydivingVectorForce=~
    HeadToPelvisDirection=~
    FloatingMultiplier=~
    FloatingAdditive=~
    DivingMultiplier=~
    DivingAdditive=~
    ParachutingMultiplier=~
    ParachutingAdditive=~
   )
   > Set vector_force(vector_force:Vector=@n4.SkydivingVectorForce) -> (Output_Get:Vector)
   n5 = Branch(execute=@n3.else, Condition="float > float"(A=Get PawnVelocityZ(), B=50) {is=title K2Node_CallFunction})
   > Set gravity_override(gravity_override="0, 0,-300")
   > Set vector_force(vector_force=@n4.SkydivingVectorForce)
   Set gravity_override(execute=@n5.else, gravity_override="0, 0,-300")
   > Set vector_force(vector_force=@n4.SkydivingVectorForce)
   > Branch(Condition=Get bCanPlayCustomAnimations() -> (bCanPlayCustomAnimations:bool))
   > Set gravity_override(gravity_override="0, 0,150")
   Set gravity_override(execute=@n2.else, gravity_override="0, 0,-980")
   > Set vector_force(vector_force="0, 0, 0")
   Event BlueprintInitializeAnimation()
   > Set Skel_Mesh(Skel_Mesh=GetOwnerSkeletalMeshComponent())
  ]}
]
variables [
 gravity_override {type=Vector}
 Skel_Mesh {type=SkeletalMeshComponent}
 thigh_rot_l {type=number}
 thigh_rot_r {type=number}
 vector_force {type=Vector}
]
