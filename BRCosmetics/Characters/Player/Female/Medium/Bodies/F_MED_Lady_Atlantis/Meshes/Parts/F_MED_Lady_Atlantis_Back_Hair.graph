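# Physics asset for the Lady Atlantis female medium character's back hair part, providing capsule bodies and joints for hair physics.
# Five hair chains (back center, back left/right, side left/right) each link three segments to a small driver body, with tight swing and tighter twist limits.
# Drivers join spine_05 through angularly free constraints.
# Larger capsules on head, neck, clavicles and upper spine approximate the body around the hair.
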PhysicsAsset /BRCosmetics/Characters/Player/Female/Medium/Bodies/F_MED_Lady_Atlantis/Meshes/Parts/F_MED_Lady_Atlantis_Back_Hair
bodies {
 clavicle_l [{shapeType=Capsule, length=7.75, radius=5.255, center=<6.674 0.757 -1.0539>, rotation=<rot 90 0 0>}]
 clavicle_r [{shapeType=Capsule, length=11, radius=5.255, center=<-4.9001 -0.9676 1.5049>, rotation=<rot 86.9711 0 0>}]
 dyn_driver_hair_bck [{shapeType=Capsule, length=1, radius=1}]
 dyn_driver_hair_bck_l [{shapeType=Capsule, length=1, radius=1}]
 dyn_driver_hair_bck_r [{shapeType=Capsule, length=0.505, radius=0.505, rotation=<rot 0 -90 0>}]
 dyn_driver_hair_side_l [{shapeType=Capsule, length=1, radius=1}]
 dyn_driver_hair_side_r [{shapeType=Capsule, length=1, radius=1}]
 dyn_hair_bck_01 [{shapeType=Capsule, length=3, radius=2.25, center=<4 0.0508 0>, rotation=<rot 90 0 0>}]
 dyn_hair_bck_01_l [{shapeType=Capsule, length=3, radius=2.25, center=<4 0 0>, rotation=<rot 90 0 0>}]
 dyn_hair_bck_01_r [{shapeType=Capsule, length=3, radius=2, center=<3 0 0>, rotation=<rot 90 0 0>}]
 dyn_hair_bck_02 [{shapeType=Capsule, length=3, radius=2.25, center=<4 0 0>, rotation=<rot 90 0 0>}]
 dyn_hair_bck_02_l [{shapeType=Capsule, length=3.25, radius=2, center=<4 0 0>, rotation=<rot 90 0 0>}]
 dyn_hair_bck_02_r [{shapeType=Capsule, length=2.25, radius=2, center=<3 0 0>, rotation=<rot 90 0 0>}]
 dyn_hair_bck_03 [{shapeType=Capsule, length=3, radius=2.25, center=<4 0 0>, rotation=<rot 90 0 0>}]
 dyn_hair_bck_03_l [{shapeType=Capsule, length=1.75, radius=2, center=<4 0 0>, rotation=<rot 90 0 0>}]
 dyn_hair_bck_03_r [{shapeType=Capsule, length=2, radius=2, center=<3 0 0>, rotation=<rot 90 0 0>}]
 dyn_hair_side_01_l [{shapeType=Capsule, length=3, radius=2.25, center=<4 0 0>, rotation=<rot 90 0 0>}]
 dyn_hair_side_01_r [{shapeType=Capsule, length=3, radius=2.25, center=<4 0 0>, rotation=<rot 90 0 0>}]
 dyn_hair_side_02_l [{shapeType=Capsule, length=3.25, radius=2, center=<4 0 0>, rotation=<rot 90 0 0>}]
 dyn_hair_side_02_r [{shapeType=Capsule, length=2.1125, radius=2.6375, center=<3.7702 0 0>, rotation=<rot 90 0 0>}]
 dyn_hair_side_03_l [{shapeType=Capsule, length=3.5, radius=1.75, center=<4 0 0>, rotation=<rot 90 0 0>}]
 dyn_hair_side_03_r [{shapeType=Capsule, length=2.75, radius=2, center=<3.1399 0 0>, rotation=<rot 90 0 0>}]
 head [{shapeType=Capsule, length=6.25, radius=5.5, center=<2.1907 1.7605 0>, rotation=<rot 90 0 7.4685>}]
 neck_01 [{shapeType=Capsule, length=5, radius=5.255, center=<2.363 0 0.0001>, rotation=<rot 90 -8.8807 -2.1695>}]
 spine_05 [{shapeType=Capsule, length=23.5, radius=15.75, center=<-1.5227 7.3375 0.0003>}]
}
constraints [
 {bone1Name=head, bone2Name=neck_01, swing1LimitDegrees=45, swing2LimitDegrees=45, twistLimitDegrees=45}
 {
  bone1Name=neck_01
  bone2Name=spine_05
  swing1LimitDegrees=45
  swing1Motion=Free
  swing2LimitDegrees=45
  swing2Motion=Free
  twistLimitDegrees=45
  twistMotion=Free
 }
 {
  bone1Name=clavicle_l
  bone2Name=spine_05
  swing1LimitDegrees=45
  swing1Motion=Free
  swing2LimitDegrees=45
  swing2Motion=Free
  twistLimitDegrees=45
  twistMotion=Free
 }
 {
  bone1Name=clavicle_r
  bone2Name=spine_05
  swing1LimitDegrees=45
  swing1Motion=Free
  swing2LimitDegrees=45
  swing2Motion=Free
  twistLimitDegrees=45
  twistMotion=Free
 }
 {
  bone1Name=dyn_driver_hair_bck
  bone2Name=spine_05
  swing1LimitDegrees=45
  swing1Motion=Free
  swing2LimitDegrees=45
  swing2Motion=Free
  twistLimitDegrees=45
  twistMotion=Free
 }
 {bone1Name=dyn_hair_bck_01, bone2Name=dyn_driver_hair_bck, swing1LimitDegrees=15, swing2LimitDegrees=15, twistLimitDegrees=5}
 {bone1Name=dyn_hair_bck_02, bone2Name=dyn_hair_bck_01, swing1LimitDegrees=15, swing2LimitDegrees=15, twistLimitDegrees=5}
 {bone1Name=dyn_hair_bck_03, bone2Name=dyn_hair_bck_02, swing1LimitDegrees=15, swing2LimitDegrees=15, twistLimitDegrees=5}
 {
  bone1Name=dyn_driver_hair_bck_l
  bone2Name=spine_05
  swing1LimitDegrees=45
  swing1Motion=Free
  swing2LimitDegrees=45
  swing2Motion=Free
  twistLimitDegrees=45
  twistMotion=Free
 }
 {bone1Name=dyn_hair_bck_01_l, bone2Name=dyn_driver_hair_bck_l, swing1LimitDegrees=15, swing2LimitDegrees=15, twistLimitDegrees=5}
 {bone1Name=dyn_hair_bck_02_l, bone2Name=dyn_hair_bck_01_l, swing1LimitDegrees=15, swing2LimitDegrees=15, twistLimitDegrees=5}
 {bone1Name=dyn_hair_bck_03_l, bone2Name=dyn_hair_bck_02_l, swing1LimitDegrees=15, swing2LimitDegrees=15, twistLimitDegrees=5}
 {
  bone1Name=dyn_driver_hair_side_l
  bone2Name=spine_05
  swing1LimitDegrees=45
  swing1Motion=Free
  swing2LimitDegrees=45
  swing2Motion=Free
  twistLimitDegrees=45
  twistMotion=Free
 }
 {bone1Name=dyn_hair_side_01_l, bone2Name=dyn_driver_hair_side_l, swing1LimitDegrees=15, swing2LimitDegrees=15, twistLimitDegrees=5}
 {bone1Name=dyn_hair_side_02_l, bone2Name=dyn_hair_side_01_l, swing1LimitDegrees=15, swing2LimitDegrees=15, twistLimitDegrees=5}
 {bone1Name=dyn_hair_side_03_l, bone2Name=dyn_hair_side_02_l, swing1LimitDegrees=15, swing2LimitDegrees=15, twistLimitDegrees=5}
 {
  bone1Name=dyn_driver_hair_side_r
  bone2Name=spine_05
  swing1LimitDegrees=45
  swing1Motion=Free
  swing2LimitDegrees=45
  swing2Motion=Free
  twistLimitDegrees=45
  twistMotion=Free
 }
 {bone1Name=dyn_hair_side_01_r, bone2Name=dyn_driver_hair_side_r, swing1LimitDegrees=15, swing2LimitDegrees=15, twistLimitDegrees=5}
 {bone1Name=dyn_hair_side_02_r, bone2Name=dyn_hair_side_01_r, swing1LimitDegrees=15, swing2LimitDegrees=15, twistLimitDegrees=5}
 {bone1Name=dyn_hair_side_03_r, bone2Name=dyn_hair_side_02_r, swing1LimitDegrees=15, swing2LimitDegrees=15, twistLimitDegrees=5}
 {
  bone1Name=dyn_driver_hair_bck_r
  bone2Name=spine_05
  swing1LimitDegrees=45
  swing1Motion=Free
  swing2LimitDegrees=45
  swing2Motion=Free
  twistLimitDegrees=45
  twistMotion=Free
 }
 {bone1Name=dyn_hair_bck_01_r, bone2Name=dyn_driver_hair_bck_r, swing1LimitDegrees=15, swing2LimitDegrees=15, twistLimitDegrees=5}
 {bone1Name=dyn_hair_bck_02_r, bone2Name=dyn_hair_bck_01_r, swing1LimitDegrees=15, swing2LimitDegrees=15, twistLimitDegrees=5}
 {bone1Name=dyn_hair_bck_03_r, bone2Name=dyn_hair_bck_02_r, swing1LimitDegrees=15, swing2LimitDegrees=15, twistLimitDegrees=5}
]
