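# Control Rig for the Lady Atlantis FaceAcc part that offsets hair and ponytail driver bones as the head turns.
# Four chains read head rotation relative to spine_05, two on Euler Z and two on X for opposite directions.
# Each remaps the angle into a clamped, time-smoothed weight for additive local bone offsets.
# One chain also applies constant ponytail parent offsets; one hair_bck_r modifier is left unconnected.
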ControlRigBlueprint /BRCosmetics/Characters/Player/Female/Medium/Bodies/F_MED_Lady_Atlantis/Meshes/Parts/F_MED_Lady_Atlantis_FaceAcc_CtrlRig
graphs [
 RigVMFunctionLibrary {links [], nodes []}
 RigVMModel {nodes [
   n1 = &RigUnit_BeginExecution()
   n2 = &RigUnit_GetRelativeBoneTransform(Bone=head, Space=spine_05)
   n3 = /Script/RigVM.RigVMFunction_AccumulateFloatLerp(
    Blend=20
    InitialValue=0
    bIntegrateDeltaTime=true
    TargetValue=&RigVMFunction_MathFloatRemap(
     SourceMaximum=-40
     SourceMinimum=0
     TargetMinimum=0
     bClamp=true
     Value=&RigVMFunction_MathQuaternionToEuler(RotationOrder=XZY, Value=@n2.Transform.Rotation).Result.Z
    ).Result
   ) {is=struct RigVMUnitNode, template="AccumulateLerp::Execute(in TargetValue,in InitialValue,in Blend,in bIntegrateDeltaTime,out Result)"}
   n4 = &RigVMFunction_MathQuaternionFromEuler(Euler=(X=0,Y=0,Z=8))
   /Script/ControlRig.RigUnit_ModifyBoneTransforms(
    BoneToModify=((Bone="dyn_driver_hair_bck_l",Transform=(Rotation=(X=0,Y=0,Z=0,W=1),Translation=(X=4,Y=-5,Z=0),Scale3D=(X=1,Y=1,Z=1))))
    Mode=AdditiveLocal
    WeightMaximum=1
    WeightMinimum=0
    Weight=@n3.Result
    BoneToModify.0.Transform.Rotation=@n4.Result
    ExecutePin=@n1.ExecutePin
   ) {is=struct RigVMUnitNode}
   > &RigUnit_ModifyBoneTransforms(
    BoneToModify=((Bone="dyn_driver_hair_bck_r",Transform=(Rotation=(X=0,Y=0,Z=0,W=1),Translation=(X=4,Y=-5,Z=0),Scale3D=(X=1,Y=1,Z=1))))
    Mode=AdditiveLocal
    WeightMaximum=1
    WeightMinimum=0
    Weight=@n3.Result
    BoneToModify.0.Transform.Rotation=@n4.Result
   )
   > &RigUnit_ModifyBoneTransforms(
    BoneToModify=((Bone="dyn_driver_hair_bck",Transform=(Rotation=(X=0,Y=0,Z=0,W=1),Translation=(X=4,Y=-5,Z=0),Scale3D=(X=1,Y=1,Z=1))))
    Mode=AdditiveLocal
    WeightMaximum=1
    WeightMinimum=0
    Weight=@n3.Result
    BoneToModify.0.Transform.Rotation=@n4.Result
   )
   n5 = &RigVMFunction_MathQuaternionFromEuler(Euler=(X=0,Y=0,Z=-5))
   > &RigUnit_ModifyBoneTransforms(
    BoneToModify=((Bone="dyn_driver_hair_side_l",Transform=(Rotation=(X=0,Y=0,Z=0,W=1),Translation=(X=2.5,Y=-2.5,Z=-1.5),Scale3D=(X=1,Y=1,Z=1))))
    Mode=AdditiveLocal
    WeightMaximum=1
    WeightMinimum=0
    Weight=@n3.Result
    BoneToModify.0.Transform.Rotation=@n5.Result
   )
   > &RigUnit_ModifyBoneTransforms(
    BoneToModify=((Bone="dyn_driver_hair_side_r",Transform=(Rotation=(X=0,Y=0,Z=0,W=1),Translation=(X=2.5,Y=-2.5,Z=1),Scale3D=(X=1,Y=1,Z=1))))
    Mode=AdditiveLocal
    WeightMaximum=1
    WeightMinimum=0
    Weight=@n3.Result
    BoneToModify.0.Transform.Rotation=@n5.Result
   )
   > &RigUnit_ModifyBoneTransforms(
    BoneToModify=((Bone="dyn_driver_pony_l",Transform=(Rotation=(X=0,Y=0,Z=0,W=1),Translation=(X=0,Y=4,Z=2),Scale3D=(X=1,Y=1,Z=1))))
    Mode=AdditiveLocal
    WeightMaximum=1
    WeightMinimum=0
    Weight=@n3.Result
    BoneToModify.0.Transform.Rotation=&RigVMFunction_MathQuaternionFromEuler(Euler=(X=0,Y=0,Z=20)).Result
   )
   > &RigUnit_ModifyBoneTransforms(
    BoneToModify=((Bone="dyn_driver_pony_r",Transform=(Rotation=(X=0,Y=0,Z=0,W=1),Translation=(X=0,Y=5,Z=-5),Scale3D=(X=1,Y=1,Z=1))))
    Mode=AdditiveLocal
    WeightMaximum=1
    WeightMinimum=0
    Weight=@n3.Result
   )
   n6 = &RigUnit_BeginExecution()
   n7 = &RigUnit_GetRelativeBoneTransform(Bone=head, Space=spine_05)
   n8 = &RigVMFunction_AccumulateFloatLerp(
    Blend=20
    InitialValue=0
    bIntegrateDeltaTime=true
    TargetValue=&RigVMFunction_MathFloatRemap(
     SourceMaximum=30
     SourceMinimum=-8
     TargetMinimum=0
     bClamp=true
     Value=&RigVMFunction_MathQuaternionToEuler(RotationOrder=XZY, Value=@n7.Transform.Rotation).Result.Z
    ).Result
   ) {template="AccumulateLerp::Execute(in TargetValue,in InitialValue,in Blend,in bIntegrateDeltaTime,out Result)"}
   &RigUnit_ModifyBoneTransforms(
    BoneToModify=((Bone="dyn_driver_hair_bck_l",Transform=(Rotation=(X=0,Y=0,Z=0,W=1),Translation=(X=-0.6,Y=0,Z=0),Scale3D=(X=1,Y=1,Z=1))))
    Mode=AdditiveLocal
    WeightMaximum=1
    WeightMinimum=0
    Weight=@n8.Result
    BoneToModify.0.Transform.Rotation=&RigVMFunction_MathQuaternionFromEuler(Euler=(X=0,Y=3,Z=4)).Result
    ExecutePin=@n6.ExecutePin
   )
   > &RigUnit_ModifyBoneTransforms(
    BoneToModify=((Bone="dyn_driver_hair_bck_r",Transform=(Rotation=(X=0,Y=0,Z=0,W=1),Translation=(X=-0.6,Y=0,Z=0),Scale3D=(X=1,Y=1,Z=1))))
    Mode=AdditiveLocal
    WeightMaximum=1
    WeightMinimum=0
    Weight=@n8.Result
    BoneToModify.0.Transform.Rotation=&RigVMFunction_MathQuaternionFromEuler(Euler=(X=0,Y=3,Z=3)).Result
   )
   > &RigUnit_ModifyBoneTransforms(
    BoneToModify=((Bone="dyn_driver_hair_bck",Transform=(Rotation=(X=0,Y=0,Z=0,W=1),Translation=(X=-0.6,Y=0,Z=0),Scale3D=(X=1,Y=1,Z=1))))
    Mode=AdditiveLocal
    WeightMaximum=1
    WeightMinimum=0
    Weight=@n8.Result
    BoneToModify.0.Transform.Rotation=&RigVMFunction_MathQuaternionFromEuler(Euler=(X=0,Y=3,Z=3.5)).Result
   )
   > &RigUnit_ModifyBoneTransforms(
    BoneToModify=((Bone="dyn_driver_hair_side_l",Transform=(Rotation=(X=0,Y=0,Z=0,W=1),Translation=(X=-0.55,Y=0,Z=0),Scale3D=(X=1,Y=1,Z=1))))
    Mode=AdditiveLocal
    WeightMaximum=1
    WeightMinimum=0
    Weight=@n8.Result
   )
   > &RigUnit_ModifyBoneTransforms(
    BoneToModify=((Bone="dyn_driver_hair_side_r",Transform=(Rotation=(X=0,Y=0,Z=0,W=1),Translation=(X=-0.55,Y=0,Z=0),Scale3D=(X=1,Y=1,Z=1))))
    Mode=AdditiveLocal
    WeightMaximum=1
    WeightMinimum=0
    Weight=@n8.Result
   )
   > &RigUnit_ModifyBoneTransforms(
    BoneToModify=((Bone="dyn_parent_pony_l",Transform=(Rotation=(X=0,Y=0,Z=0,W=1),Translation=(X=-1,Y=1,Z=0),Scale3D=(X=1,Y=1,Z=1))))
    Mode=AdditiveLocal
    Weight=1
    WeightMaximum=1
    WeightMinimum=0
   )
   > &RigUnit_ModifyBoneTransforms(
    BoneToModify=((Bone="dyn_parent_pony_r",Transform=(Rotation=(X=0,Y=0,Z=0,W=1),Translation=(X=-1,Y=1,Z=-1),Scale3D=(X=1,Y=1,Z=1))))
    Mode=AdditiveLocal
    Weight=1
    WeightMaximum=1
    WeightMinimum=0
   )
   n9 = &RigUnit_BeginExecution()
   n10 = &RigUnit_GetRelativeBoneTransform(Bone=head, Space=spine_05)
   n11 = &RigVMFunction_AccumulateFloatLerp(
    Blend=10
    InitialValue=0
    bIntegrateDeltaTime=true
    TargetValue=&RigVMFunction_MathFloatRemap(
     SourceMaximum=-50
     SourceMinimum=10
     TargetMinimum=0
     bClamp=true
     Value=&RigVMFunction_MathQuaternionToEuler(RotationOrder=XZY, Value=@n10.Transform.Rotation).Result.X
    ).Result
   ) {template="AccumulateLerp::Execute(in TargetValue,in InitialValue,in Blend,in bIntegrateDeltaTime,out Result)"}
   &RigUnit_ModifyBoneTransforms(
    BoneToModify=((Bone="dyn_driver_hair_bck_l",Transform=(Rotation=(X=0,Y=0,Z=0,W=1),Translation=(X=0,Y=-1,Z=-4.2),Scale3D=(X=1,Y=1,Z=1))))
    Mode=AdditiveLocal
    WeightMaximum=1
    WeightMinimum=0
    Weight=@n11.Result
    BoneToModify.0.Transform.Rotation=&RigVMFunction_MathQuaternionFromEuler(Euler=(X=2,Y=0,Z=2.5)).Result
    ExecutePin=@n9.ExecutePin
   )
   > &RigUnit_ModifyBoneTransforms(
    BoneToModify=((Bone="dyn_driver_hair_bck_r",Transform=(Rotation=(X=0,Y=0,Z=0,W=1),Translation=(X=0,Y=-2,Z=-5),Scale3D=(X=1,Y=1,Z=1))))
    Mode=AdditiveLocal
    WeightMaximum=1
    WeightMinimum=0
    Weight=@n11.Result
    BoneToModify.0.Transform.Rotation=&RigVMFunction_MathQuaternionFromEuler(Euler=(X=1,Y=0,Z=5)).Result
   )
   > &RigUnit_ModifyBoneTransforms(
    BoneToModify=((Bone="dyn_driver_hair_bck",Transform=(Rotation=(X=0,Y=0,Z=0,W=1),Translation=(X=0,Y=-1.5,Z=-4.65),Scale3D=(X=1,Y=1,Z=1))))
    Mode=AdditiveLocal
    WeightMaximum=1
    WeightMinimum=0
    Weight=@n11.Result
    BoneToModify.0.Transform.Rotation=&RigVMFunction_MathQuaternionFromEuler(Euler=(X=1.25,Y=0,Z=3)).Result
   )
   > &RigUnit_ModifyBoneTransforms(
    BoneToModify=((Bone="dyn_driver_hair_side_l",Transform=(Rotation=(X=0,Y=0,Z=0,W=1),Translation=(X=0,Y=-1.5,Z=-4),Scale3D=(X=1,Y=1,Z=1))))
    Mode=AdditiveLocal
    WeightMaximum=1
    WeightMinimum=0
    Weight=@n11.Result
   )
   > &RigUnit_ModifyBoneTransforms(
    BoneToModify=((Bone="dyn_driver_hair_side_r",Transform=(Rotation=(X=0,Y=0,Z=0,W=1),Translation=(X=0,Y=5,Z=-3),Scale3D=(X=1,Y=1,Z=1))))
    Mode=AdditiveLocal
    WeightMaximum=1
    WeightMinimum=0
    Weight=@n11.Result
   )
   n12 = &RigUnit_BeginExecution()
   n13 = &RigUnit_GetRelativeBoneTransform(Bone=head, Space=spine_05)
   n14 = &RigVMFunction_AccumulateFloatLerp(
    Blend=10
    InitialValue=0
    bIntegrateDeltaTime=true
    TargetValue=&RigVMFunction_MathFloatRemap(
     SourceMaximum=50
     SourceMinimum=-10
     TargetMinimum=0
     bClamp=true
     Value=&RigVMFunction_MathQuaternionToEuler(RotationOrder=XZY, Value=@n13.Transform.Rotation).Result.X
    ).Result
   ) {template="AccumulateLerp::Execute(in TargetValue,in InitialValue,in Blend,in bIntegrateDeltaTime,out Result)"}
   &RigUnit_ModifyBoneTransforms(
    BoneToModify=((Bone="dyn_driver_hair_bck_l",Transform=(Rotation=(X=0,Y=0,Z=0,W=1),Translation=(X=0,Y=-1,Z=5),Scale3D=(X=1,Y=1,Z=1))))
    Mode=AdditiveLocal
    WeightMaximum=1
    WeightMinimum=0
    Weight=@n14.Result
    BoneToModify.0.Transform.Rotation=&RigVMFunction_MathQuaternionFromEuler(Euler=(X=-2,Y=0,Z=-2.5)).Result
    ExecutePin=@n12.ExecutePin
   )
   > &RigUnit_ModifyBoneTransforms(
    BoneToModify=((Bone="dyn_driver_hair_bck_r",Transform=(Rotation=(X=0,Y=0,Z=0,W=1),Translation=(X=0,Y=-2,Z=3),Scale3D=(X=1,Y=1,Z=1))))
    Mode=AdditiveLocal
    WeightMaximum=1
    WeightMinimum=0
    Weight=@n14.Result
    BoneToModify.0.Transform.Rotation=&RigVMFunction_MathQuaternionFromEuler(Euler=(X=-1,Y=0,Z=-5)).Result
   )
   > &RigUnit_ModifyBoneTransforms(
    BoneToModify=((Bone="dyn_driver_hair_bck",Transform=(Rotation=(X=0,Y=0,Z=0,W=1),Translation=(X=0,Y=-1.5,Z=4),Scale3D=(X=1,Y=1,Z=1))))
    Mode=AdditiveLocal
    WeightMaximum=1
    WeightMinimum=0
    Weight=@n14.Result
    BoneToModify.0.Transform.Rotation=&RigVMFunction_MathQuaternionFromEuler(Euler=(X=-1.25,Y=0,Z=-3)).Result
   )
   > &RigUnit_ModifyBoneTransforms(
    BoneToModify=((Bone="dyn_driver_hair_side_l",Transform=(Rotation=(X=0,Y=0,Z=0,W=1),Translation=(X=0,Y=2,Z=4),Scale3D=(X=1,Y=1,Z=1))))
    Mode=AdditiveLocal
    WeightMaximum=1
    WeightMinimum=0
    Weight=@n14.Result
    BoneToModify.0.Transform.Rotation=&RigVMFunction_MathQuaternionFromEuler(Euler=(X=0,Y=0,Z=5)).Result
   )
   > &RigUnit_ModifyBoneTransforms(
    BoneToModify=((Bone="dyn_driver_hair_side_r",Transform=(Rotation=(X=0,Y=0,Z=0,W=1),Translation=(X=0,Y=0,Z=1),Scale3D=(X=1,Y=1,Z=1))))
    Mode=AdditiveLocal
    WeightMaximum=1
    WeightMinimum=0
    Weight=@n14.Result
   )
   &RigUnit_ModifyBoneTransforms(
    BoneToModify=((Bone="dyn_driver_hair_bck_r",Transform=(Rotation=(X=0,Y=0,Z=0,W=1),Translation=(X=2.5,Y=-1,Z=0),Scale3D=(X=1,Y=1,Z=1))))
    ExecutePin=()
    Mode=AdditiveLocal
    Weight=1
    WeightMaximum=1
    WeightMinimum=0
   )
  ]}
]
hierarchy [
 root {loc=<0 0 0>, rot=<0 0 0 1>}
 pelvis {parent=root, loc=<0 -0.5301 98.0571>, rot=<0.0127 0.707 -0.0127 -0.707>}
 spine_01 {parent=pelvis, loc=<6.1208 -0.9328 0>, rot=<0 0 0.034 -0.9994>}
 spine_02 {parent=spine_01, loc=<6.5996 -0.2095 0>, rot=<0 0 0.0238 -0.9997>}
 spine_03 {parent=spine_02, loc=<6.4914 -0.1237 0>, rot=<0 0 0.0149 -0.9999>}
 spine_04 {parent=spine_03, loc=<6.9598 0.0794 0>, rot=<0 0 -0.0132 -0.9999>}
 spine_05 {parent=spine_04, loc=<6.9987 0 0>, rot=<0 0 -0.0015 -1>}
 clavicle_l {parent=spine_05, loc=<5.4831 -0.1053 -3.2946>, rot=<0.0465 -0.7289 -0.009 -0.6829>}
 clavicle_r {parent=spine_05, loc=<5.4831 -0.1053 3.2946>, rot=<0.7289 0.0465 -0.6829 0.009>}
 neck_01 {parent=spine_05, loc=<11.3138 0.1529 0>, rot=<0 0 -0.2003 -0.9797>}
 neck_02 {parent=neck_01, loc=<4.5627 0 0>, rot=<0 0 0.0504 -0.9987>}
 head {parent=neck_02, loc=<5.3914 0 0>, rot=<0 0 0.115 -0.9934>}
 dyn_hair_tuftA_01 {parent=head, loc=<17.2628 -3.8592 0.0942>, rot=<0.7048 -0.7083 -0.0253 -0.0322>}
 dyn_hair_tuftA_02 {parent=dyn_hair_tuftA_01, loc=<3.3206 0.1358 -0.2823>, rot=<0 0 -0.0937 -0.9956>}
 dyn_hair_tuftB_01 {parent=head, loc=<15.652 -3.7328 -1.737>, rot=<0.6635 -0.736 -0.098 -0.0921>}
 dyn_hair_tuftB_02 {parent=dyn_hair_tuftB_01, loc=<3.3206 0.1357 -0.2823>, rot=<0.008 -0.0625 -0.127 -0.9899>}
 dyn_hair_tuftC_01 {parent=head, loc=<15.5545 -3.9037 2.0292>, rot=<0.6373 -0.767 0.0637 0.0374>}
 dyn_hair_tuftC_02 {parent=dyn_hair_tuftC_01, loc=<3.3206 0.1357 -0.2823>, rot=<0 0 -0.0415 -0.9991>}
 dyn_hair_top_frt_01_l {parent=head, loc=<9.4149 4.8465 -6.9056>, rot=<-0.0741 0.6494 0.7556 0.0418>}
 dyn_hair_top_frt_02_l {parent=dyn_hair_top_frt_01_l, loc=<5.9963 0 0>, rot=<0.0272 -0.0575 -0.0081 -0.9979>}
 dyn_hair_top_frt_03_l {parent=dyn_hair_top_frt_02_l, loc=<5.4153 0 0>, rot=<-0.0436 0 0 -0.999>}
 dyn_hair_top_side_01_l {parent=head, loc=<9.926 -1.2392 -8.2198>, rot=<0.1137 -0.7536 -0.6474 0.0021>}
 dyn_hair_top_side_02_l {parent=dyn_hair_top_side_01_l, loc=<5.9963 0 0>, rot=<0 0 0 1>}
 dyn_hair_top_bck_01_l {parent=head, loc=<10.7143 -7.5523 -4.6495>, rot=<0.0479 -0.9853 -0.1626 0.0214>}
 dyn_hair_top_bck_02_l {parent=dyn_hair_top_bck_01_l, loc=<6.0505 0 0>, rot=<0 0 0.0233 -0.9997>}
 dyn_hair_top_bck_03_l {parent=dyn_hair_top_bck_02_l, loc=<5.2581 0 0>, rot=<0 0 0 -1>}
 dyn_hair_top_bck_01 {parent=head, loc=<10.97 -9.5439 0.0963>, rot=<0.0426 -0.9976 0.0538 0.0014>}
 dyn_hair_top_bck_02 {parent=dyn_hair_top_bck_01, loc=<5.9306 0 0>, rot=<0 0 0 -1>}
 dyn_hair_top_bck_03 {parent=dyn_hair_top_bck_02, loc=<5.668 0 0>, rot=<0 0 0 -1>}
 dyn_hair_top_frt_01_r {parent=head, loc=<9.4342 4.5179 6.954>, rot=<0.0961 -0.6504 0.7532 0.0228>}
 dyn_hair_top_frt_02_r {parent=dyn_hair_top_frt_01_r, loc=<5.9963 0 0>, rot=<-0.0272 0.0575 -0.0081 -0.9979>}
 dyn_hair_top_side_01_r {parent=head, loc=<9.9263 -1.2392 8.2198>, rot=<0.1137 -0.7536 0.6474 -0.0021>}
 dyn_hair_top_side_02_r {parent=dyn_hair_top_side_01_r, loc=<5.9968 0.0001 -0.0001>, rot=<0 0.044 0 -0.999>}
 dyn_hair_top_bck_01_r {parent=head, loc=<10.7146 -7.5523 4.6495>, rot=<0.0416 -0.9872 0.153 0.0177>}
 dyn_hair_top_bck_02_r {parent=dyn_hair_top_bck_01_r, loc=<6.0504 0 0>, rot=<0 0 0.0233 -0.9997>}
 dyn_hair_top_bck_03_r {parent=dyn_hair_top_bck_02_r, loc=<5.2579 0 0>, rot=<0 0 0 1>}
 dyn_parent_hairy_l {parent=head, loc=<5.6259 1.1037 -7.0827>, rot=<0.7071 0.0029 -0.7071 0.0029>}
 dyn_driver_hairy_l {parent=dyn_parent_hairy_l, loc=<0 0 0>, rot=<-0.4533 -0.4998 0.5154 -0.5282>}
 dyn_hairy_01_l {parent=dyn_driver_hairy_l, loc=<0 0 0>, rot=<0 0 0 -1>}
 dyn_parent_hairy_r {parent=head, loc=<5.6257 1.1037 6.7963>, rot=<0.0029 -0.7071 -0.0029 -0.7071>}
 dyn_driver_hairy_r {parent=dyn_parent_hairy_r, loc=<0 0 0>, rot=<-0.4593 -0.506 0.5101 -0.5224>}
 dyn_hairy_01_r {parent=dyn_driver_hairy_r, loc=<0 0 0>, rot=<0 0 0 -1>}
 dyn_parent_hair_bck_r {parent=spine_05, loc=<21.4929 -1.9778 3.8919>, rot=<0.1927 -0.9792 0.0444 0.0443>}
 dyn_driver_hair_bck_r {parent=dyn_parent_hair_bck_r, loc=<0 0 0>, rot=<0 0 0 1>}
 dyn_hair_bck_01_r {parent=dyn_driver_hair_bck_r, loc=<0 0 0>, rot=<0 0 0 1>}
 dyn_hair_bck_02_r {parent=dyn_hair_bck_01_r, loc=<5.9306 0 0>, rot=<0 0 0.1156 -0.9933>}
 dyn_hair_bck_03_r {parent=dyn_hair_bck_02_r, loc=<5.668 0 0>, rot=<0 0 -0.0708 -0.9975>}
 dyn_parent_hair_bck {parent=spine_05, loc=<21.3345 -2.6666 0.0125>, rot=<0.1911 -0.9816 -0.0019 0.0036>}
 dyn_driver_hair_bck {parent=dyn_parent_hair_bck, loc=<0 0 0>, rot=<0 0 0 -1>}
 dyn_hair_bck_01 {parent=dyn_driver_hair_bck, loc=<0 0 0>, rot=<0 0 0 -1>}
 dyn_hair_bck_02 {parent=dyn_hair_bck_01, loc=<5.9306 0 0>, rot=<0 0 0.1156 -0.9933>}
 dyn_hair_bck_03 {parent=dyn_hair_bck_02, loc=<5.668 0 0>, rot=<0 0 -0.0708 -0.9975>}
 dyn_parent_hair_bck_l {parent=spine_05, loc=<21.4934 -1.9744 -3.9054>, rot=<0.1946 -0.9785 -0.0642 -0.0238>}
 dyn_driver_hair_bck_l {parent=dyn_parent_hair_bck_l, loc=<0 0 0>, rot=<0 0 0 1>}
 dyn_hair_bck_01_l {parent=dyn_driver_hair_bck_l, loc=<0 0 0>, rot=<0 0 0 1>}
 dyn_hair_bck_02_l {parent=dyn_hair_bck_01_l, loc=<5.9306 0 0>, rot=<0 0 0.1156 -0.9933>}
 dyn_hair_bck_03_l {parent=dyn_hair_bck_02_l, loc=<5.6679 0 0>, rot=<0 0 -0.0708 -0.9975>}
 dyn_parent_hair_side_l {parent=spine_05, loc=<22.2875 0.7973 -6.4773>, rot=<0.1581 -0.958 -0.2329 -0.0542>}
 dyn_driver_hair_side_l {parent=dyn_parent_hair_side_l, loc=<0 0 0>, rot=<0 0 0 1>}
 dyn_hair_side_01_l {parent=dyn_driver_hair_side_l, loc=<0 0 0>, rot=<0 0 0 1>}
 dyn_hair_side_02_l {parent=dyn_hair_side_01_l, loc=<5.9306 0 0>, rot=<0 0 0.1493 -0.9888>}
 dyn_hair_side_03_l {parent=dyn_hair_side_02_l, loc=<5.0013 0.0917 0>, rot=<0 0 0.0683 -0.9977>}
 dyn_parent_hair_side_r {parent=spine_05, loc=<22.2878 0.7974 6.4773>, rot=<0.1617 -0.9588 0.2304 0.0396>}
 dyn_driver_hair_side_r {parent=dyn_parent_hair_side_r, loc=<0 0 0>, rot=<0 0 0 -1>}
 dyn_hair_side_01_r {parent=dyn_driver_hair_side_r, loc=<0 0 0>, rot=<0 0 0 -1>}
 dyn_hair_side_02_r {parent=dyn_hair_side_01_r, loc=<5.9305 0 0>, rot=<0 0 0.1493 -0.9888>}
 dyn_hair_side_03_r {parent=dyn_hair_side_02_r, loc=<5.0016 0.0919 0>, rot=<0 0 0.0683 -0.9977>}
 dyn_parent_pony_r {parent=spine_04, loc=<22.899 3.5784 7.1845>, rot=<0.0294 -0.7065 -0.0294 -0.7065>}
 dyn_driver_pony_r {parent=dyn_parent_pony_r, loc=<0 0 0>, rot=<-0.6717 -0.0088 0.7383 0.0597>}
 dyn_pony_01_r {parent=dyn_driver_pony_r, loc=<0 0 0>, rot=<-0.0054 -0.0115 -0.0014 -0.9999>}
 dyn_pony_02_r {parent=dyn_pony_01_r, loc=<4.5036 0 0>, rot=<-0.0239 -0.1176 0.3004 -0.9462>}
 dyn_pony_03_r {parent=dyn_pony_02_r, loc=<4.1985 0 0>, rot=<-0.0495 -0.1034 0.145 -0.9828>}
 dyn_parent_pony_l {parent=spine_04, loc=<22.9205 3.325 -7.3973>, rot=<0.0294 -0.7065 -0.0294 -0.7065>}
 dyn_driver_pony_l {parent=dyn_parent_pony_l, loc=<0 0 0>, rot=<-0.0702 0.7374 0.0184 -0.6715>}
 dyn_pony_01_l {parent=dyn_driver_pony_l, loc=<0 0 0>, rot=<-0.0054 -0.0115 -0.0014 -0.9999>}
 dyn_pony_02_l {parent=dyn_pony_01_l, loc=<-4.5041 0 0>, rot=<-0.0138 -0.0668 0.2699 -0.9605>}
 dyn_pony_03_l {parent=dyn_pony_02_l, loc=<-4.1981 0.0004 0.0001>, rot=<-0.0532 -0.0744 0.0882 -0.9919>}
]
