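# Physics asset for the front tails of the female Lady Atlantis player cosmetic, covering its left and right dynamic pony bone chains.
# Each side chains three small capsules from a tiny driver body, with tight swing and tighter twist limits at each link.
# Larger capsules on head, neck, upper spine and clavicles form a constrained upper-body group, apparently as collision for the tails.
# No constraint joins the driver bodies to that group.
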PhysicsAsset /BRCosmetics/Characters/Player/Female/Medium/Bodies/F_MED_Lady_Atlantis/Meshes/Parts/F_MED_Lady_Atlantis_FrontTails
bodies {
 clavicle_l [{shapeType=Capsule, length=10, radius=5, center=<9.7308 -0.8382 0>, rotation=<rot 90 0 0>}]
 clavicle_r [{shapeType=Capsule, length=12.75, radius=5, center=<-8.0891 0.6195 -0.0108>, rotation=<rot 90 0 0>}]
 dyn_driver_pony_l [{shapeType=Capsule, length=0.505, radius=0.505, rotation=<rot 0 -90 0>}]
 dyn_driver_pony_r [{shapeType=Capsule, length=0.505, radius=0.505, rotation=<rot 0 -90 0>}]
 dyn_pony_01_l [{shapeType=Capsule, length=3, radius=2, center=<4.6709 -0.5 0>, rotation=<rot 90 0 29.6369>}]
 dyn_pony_01_r [{shapeType=Capsule, length=3, radius=2, center=<5 -0.5 0>, rotation=<rot 90 -90 -67.0596>}]
 dyn_pony_02_l [{shapeType=Capsule, length=3, radius=2, center=<5 -1 0>, rotation=<rot 90 7.125 17.1931>}]
 dyn_pony_02_r [{shapeType=Capsule, length=3, radius=2, center=<5 -1 0>, rotation=<rot 90 7.125 16.0872>}]
 dyn_pony_03_l [{shapeType=Capsule, length=3, radius=2, center=<5 -0.3076 0>, rotation=<rot 90 0 0>}]
 dyn_pony_03_r [{shapeType=Capsule, length=3, radius=2, center=<5 -0.2108 0>, rotation=<rot 90 0 0>}]
 head [{shapeType=Capsule, length=9.5, radius=7, center=<2.4171 -1.2461 -0.0001>, rotation=<rot 90 0 -0.0795>}]
 neck_01 [{shapeType=Capsule, length=5, radius=5.255, center=<2.363 0 0.0001>, rotation=<rot 90 -8.8807 -2.1695>}]
 spine_05 [{shapeType=Capsule, length=16, radius=18.25, center=<-7.3686 -7.6585 0.0004>, rotation=<rot 0 -7.1772 0>}]
}
constraints [
 {bone1Name=head, bone2Name=neck_01, swing1LimitDegrees=45, swing2LimitDegrees=45, twistLimitDegrees=45}
 {
  bone1Name=neck_01
  bone2Name=spine_05
  swing1LimitDegrees=45
  swing1Motion=Free
  swing2LimitDegrees=45
  swing2Motion=Free
  twistLimitDegrees=45
  twistMotion=Free
 }
 {
  bone1Name=clavicle_l
  bone2Name=spine_05
  swing1LimitDegrees=45
  swing1Motion=Free
  swing2LimitDegrees=45
  swing2Motion=Free
  twistLimitDegrees=45
  twistMotion=Free
 }
 {
  bone1Name=clavicle_r
  bone2Name=spine_05
  swing1LimitDegrees=45
  swing1Motion=Free
  swing2LimitDegrees=45
  swing2Motion=Free
  twistLimitDegrees=45
  twistMotion=Free
 }
 {bone1Name=dyn_pony_01_r, bone2Name=dyn_driver_pony_r, swing1LimitDegrees=15, swing2LimitDegrees=15, twistLimitDegrees=5}
 {bone1Name=dyn_pony_02_r, bone2Name=dyn_pony_01_r, swing1LimitDegrees=15, swing2LimitDegrees=15, twistLimitDegrees=5}
 {bone1Name=dyn_pony_03_r, bone2Name=dyn_pony_02_r, swing1LimitDegrees=15, swing2LimitDegrees=15, twistLimitDegrees=5}
 {bone1Name=dyn_pony_01_l, bone2Name=dyn_driver_pony_l, swing1LimitDegrees=15, swing2LimitDegrees=15, twistLimitDegrees=5}
 {bone1Name=dyn_pony_02_l, bone2Name=dyn_pony_01_l, swing1LimitDegrees=15, swing2LimitDegrees=15, twistLimitDegrees=5}
 {bone1Name=dyn_pony_03_l, bone2Name=dyn_pony_02_l, swing1LimitDegrees=15, swing2LimitDegrees=15, twistLimitDegrees=5}
]
