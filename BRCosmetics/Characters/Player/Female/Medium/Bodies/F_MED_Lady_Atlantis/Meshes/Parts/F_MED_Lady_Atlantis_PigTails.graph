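# Physics asset for the pigtails of the Lady Atlantis female medium body cosmetic, giving the two hair strands capsule bodies and constraints.
# Head, neck, upper spine and clavicle capsules form the body volumes around the hair.
# Each side chains a small driver body to the head with free rotation, then one pigtail body to that driver with tight swing and twist limits.
# Neck and clavicles rotate freely against the spine.
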PhysicsAsset /BRCosmetics/Characters/Player/Female/Medium/Bodies/F_MED_Lady_Atlantis/Meshes/Parts/F_MED_Lady_Atlantis_PigTails
bodies {
 clavicle_l [{shapeType=Capsule, length=10, radius=5, center=<9.7309 0.0468 0>, rotation=<rot 90 0 0>}]
 clavicle_r [{shapeType=Capsule, length=12.75, radius=5, center=<-8.0002 -0.261 -0.0048>, rotation=<rot 90 0 0>}]
 dyn_driver_hairy_l [{shapeType=Capsule, length=1, radius=1}]
 dyn_driver_hairy_r [{shapeType=Capsule, length=1, radius=1}]
 dyn_hairy_01_l [{shapeType=Capsule, length=6.75, radius=2, center=<5 -1.25 0>, rotation=<rot 90 0 0>}]
 dyn_hairy_01_r [{shapeType=Capsule, length=6.75, radius=2, center=<5 -1.25 0>, rotation=<rot 90 0 0>}]
 head [
  {shapeType=Capsule, length=6.25, radius=5.5, center=<2.3198 2.7452 -0.0001>, rotation=<rot 90 0 7.4684>}
  {shapeType=Capsule, length=9.5, radius=7, center=<2.4171 -1.2461 -0.0001>, rotation=<rot 90 0 -0.0795>}
 ]
 neck_01 [{shapeType=Capsule, length=5, radius=5.255, center=<2.363 0 0.0001>, rotation=<rot 90 -8.8807 -2.1695>}]
 spine_05 [{shapeType=Capsule, length=15.75, radius=18.5, center=<-3.3929 -9.6141 0.0003>}]
}
constraints [
 {bone1Name=head, bone2Name=neck_01, swing1LimitDegrees=45, swing2LimitDegrees=45, twistLimitDegrees=45}
 {
  bone1Name=neck_01
  bone2Name=spine_05
  swing1LimitDegrees=45
  swing1Motion=Free
  swing2LimitDegrees=45
  swing2Motion=Free
  twistLimitDegrees=45
  twistMotion=Free
 }
 {
  bone1Name=clavicle_l
  bone2Name=spine_05
  swing1LimitDegrees=45
  swing1Motion=Free
  swing2LimitDegrees=45
  swing2Motion=Free
  twistLimitDegrees=45
  twistMotion=Free
 }
 {
  bone1Name=clavicle_r
  bone2Name=spine_05
  swing1LimitDegrees=45
  swing1Motion=Free
  swing2LimitDegrees=45
  swing2Motion=Free
  twistLimitDegrees=45
  twistMotion=Free
 }
 {
  bone1Name=dyn_driver_hairy_l
  bone2Name=head
  swing1LimitDegrees=45
  swing1Motion=Free
  swing2LimitDegrees=45
  swing2Motion=Free
  twistLimitDegrees=45
  twistMotion=Free
 }
 {bone1Name=dyn_hairy_01_l, bone2Name=dyn_driver_hairy_l, swing1LimitDegrees=15, swing2LimitDegrees=20, twistLimitDegrees=5}
 {
  bone1Name=dyn_driver_hairy_r
  bone2Name=head
  swing1LimitDegrees=45
  swing1Motion=Free
  swing2LimitDegrees=45
  swing2Motion=Free
  twistLimitDegrees=45
  twistMotion=Free
 }
 {bone1Name=dyn_hairy_01_r, bone2Name=dyn_driver_hairy_r, swing1LimitDegrees=15, swing2LimitDegrees=20, twistLimitDegrees=5}
]
