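# Physics asset for the hair of the female medium Lady Atlantis SkyTop character part, for simulating its dyn_hair bones.
# Single capsules cover each hair bone, plus head, neck_01 and spine_05 as body volumes.
# Tuft and top-hair chains of two or three bones hang from head with tight swing and twist limits, looser on the right front and side.
# Single-bone back and side strands constrain to spine_05, not head, with free swing and twist.
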PhysicsAsset /BRCosmetics/Characters/Player/Female/Medium/Bodies/F_MED_Lady_Atlantis/Meshes/Parts/F_MED_Lady_Atlantis_SkyTop_Hair1
bodies {
 dyn_hair_bck_01 [{shapeType=Capsule, length=4.25, radius=2.05, center=<5.9117 0.4344 0>, rotation=<rot 90 0 14.2332>}]
 dyn_hair_bck_01_l [{shapeType=Capsule, length=3.75, radius=2.05, center=<6.4758 0.2167 0>, rotation=<rot 90 -7.125 5.1638>}]
 dyn_hair_bck_01_r [{shapeType=Capsule, length=3.25, radius=2.05, center=<6.0893 0.4194 0>, rotation=<rot 90 7.125 24.7901>}]
 dyn_hair_side_01_l [{shapeType=Capsule, length=5.75, radius=2.05, center=<3.5 0 0>, rotation=<rot 90 0 0>}]
 dyn_hair_side_01_r [{shapeType=Capsule, length=5.75, radius=2.05, center=<3.5 0 0>, rotation=<rot 90 0 0>}]
 dyn_hair_top_bck_01 [{shapeType=Capsule, length=3, radius=1.75, center=<3 -0.0455 0>, rotation=<rot 90 0 0>}]
 dyn_hair_top_bck_01_l [{shapeType=Capsule, length=3, radius=1.75, center=<3 -0.9 0>, rotation=<rot 90 0 0>}]
 dyn_hair_top_bck_01_r [{shapeType=Capsule, length=3, radius=2, center=<3 -0.9 0>, rotation=<rot 90 0 0>}]
 dyn_hair_top_bck_02 [{shapeType=Capsule, length=1.75, radius=2.25, center=<2.6534 0 0>, rotation=<rot 90 0 0>}]
 dyn_hair_top_bck_02_l [{shapeType=Capsule, length=2.25, radius=1.75, center=<2.0626 0 0>, rotation=<rot 90 0 0>}]
 dyn_hair_top_bck_02_r [{shapeType=Capsule, length=3, radius=2, center=<2.4011 -0.1905 0>, rotation=<rot 90 0 0>}]
 dyn_hair_top_bck_03 [{shapeType=Capsule, length=1.75, radius=2.5, center=<1.607 -0.5 -0.55>, rotation=<rot 90 -90 -70.9995>}]
 dyn_hair_top_bck_03_l [{shapeType=Capsule, length=2.5, radius=2, center=<1.8306 -0.53 -0.025>, rotation=<rot 79.9982 179.9999 -162.1443>}]
 dyn_hair_top_bck_03_r [{shapeType=Capsule, length=3, radius=2, center=<1 -1 0>, rotation=<rot 90 0 23.2584>}]
 dyn_hair_top_frt_01_l [{shapeType=Capsule, length=3.25, radius=1.75, center=<3 -0.1943 0>, rotation=<rot 90 0 0>}]
 dyn_hair_top_frt_01_r [{shapeType=Capsule, length=3.5, radius=1.5, center=<3 0 0>, rotation=<rot 90 0 0>}]
 dyn_hair_top_frt_02_l [{shapeType=Capsule, length=2.25, radius=1.75, center=<3 -0.1 0>, rotation=<rot 90 0 0>}]
 dyn_hair_top_frt_02_r [{shapeType=Capsule, length=1.75, radius=1.75, center=<3.25 0 1>, rotation=<rot 60 -180 -180>}]
 dyn_hair_top_frt_03_l [{shapeType=Capsule, length=1.25, radius=2, center=<1.25 -0.65 -0.45>, rotation=<rot 72.917 0.0001 19.7158>}]
 dyn_hair_top_side_01_l [{shapeType=Capsule, length=2.75, radius=1.75, center=<3 -0.4 0>, rotation=<rot 90 -3.5763 -7.2776>}]
 dyn_hair_top_side_01_r [{shapeType=Capsule, length=3.25, radius=1.75, center=<3 -0.25 0>, rotation=<rot 90 -176.4237 -179.8182>}]
 dyn_hair_top_side_02_l [{shapeType=Capsule, length=2.25, radius=1.75, center=<2.5 0 -0.75>, rotation=<rot 63.0994 0 0>}]
 dyn_hair_top_side_02_r [{shapeType=Capsule, length=2.5, radius=1.5, center=<3.25 0 0.45>, rotation=<rot 69.0892 179.9998 -180>}]
 dyn_hair_tuftA_01 [{shapeType=Capsule, length=2, radius=1, center=<2 0 0>, rotation=<rot 90 0 0>}]
 dyn_hair_tuftA_02 [{shapeType=Capsule, length=2, radius=1, center=<2 0 0>, rotation=<rot 90 0 0>}]
 dyn_hair_tuftB_01 [{shapeType=Capsule, length=2, radius=1, center=<0.1 0 0>, rotation=<rot 90 0 0>}]
 dyn_hair_tuftB_02 [{shapeType=Capsule, length=2, radius=1, center=<0.1 0 0>, rotation=<rot 90 0 0>}]
 dyn_hair_tuftC_01 [{shapeType=Capsule, length=2, radius=1, center=<2 0 0>, rotation=<rot 90 0 0>}]
 dyn_hair_tuftC_02 [{shapeType=Capsule, length=2, radius=1, center=<2 0 0>, rotation=<rot 90 0 0>}]
 head [{shapeType=Capsule, length=9, radius=7, center=<1.9574 -0.0193 -0.0001>, rotation=<rot 90 0 7.4685>}]
 neck_01 [{shapeType=Capsule, length=5, radius=5.255, center=<2.363 0 0.0001>, rotation=<rot 90 -8.8807 -2.1695>}]
 spine_05 [{shapeType=Capsule, length=19, radius=20.25, center=<-8.3135 4.9469 0.0002>}]
}
constraints [
 {bone1Name=head, bone2Name=neck_01, swing1LimitDegrees=45, swing2LimitDegrees=45, twistLimitDegrees=45}
 {
  bone1Name=neck_01
  bone2Name=spine_05
  swing1LimitDegrees=45
  swing1Motion=Free
  swing2LimitDegrees=45
  swing2Motion=Free
  twistLimitDegrees=45
  twistMotion=Free
 }
 {bone1Name=dyn_hair_tuftA_01, bone2Name=head, swing1LimitDegrees=15, swing2LimitDegrees=15, twistLimitDegrees=5}
 {bone1Name=dyn_hair_tuftA_02, bone2Name=dyn_hair_tuftA_01, swing1LimitDegrees=15, swing2LimitDegrees=15, twistLimitDegrees=5}
 {bone1Name=dyn_hair_tuftB_01, bone2Name=head, swing1LimitDegrees=15, swing2LimitDegrees=20, twistLimitDegrees=5}
 {bone1Name=dyn_hair_tuftB_02, bone2Name=dyn_hair_tuftB_01, swing1LimitDegrees=20, swing2LimitDegrees=35, twistLimitDegrees=5}
 {bone1Name=dyn_hair_tuftC_01, bone2Name=head, swing1LimitDegrees=15, swing2LimitDegrees=25, twistLimitDegrees=5}
 {bone1Name=dyn_hair_tuftC_02, bone2Name=dyn_hair_tuftC_01, swing1LimitDegrees=20, swing2LimitDegrees=35, twistLimitDegrees=5}
 {bone1Name=dyn_hair_top_frt_01_l, bone2Name=head, swing1LimitDegrees=15, swing2LimitDegrees=15, twistLimitDegrees=5}
 {bone1Name=dyn_hair_top_frt_02_l, bone2Name=dyn_hair_top_frt_01_l, swing1LimitDegrees=15, swing2LimitDegrees=15, twistLimitDegrees=5}
 {bone1Name=dyn_hair_top_frt_03_l, bone2Name=dyn_hair_top_frt_02_l, swing1LimitDegrees=15, swing2LimitDegrees=15, twistLimitDegrees=5}
 {bone1Name=dyn_hair_top_side_01_l, bone2Name=head, swing1LimitDegrees=15, swing2LimitDegrees=15, twistLimitDegrees=5}
 {bone1Name=dyn_hair_top_side_02_l, bone2Name=dyn_hair_top_side_01_l, swing1LimitDegrees=18, swing2LimitDegrees=18, twistLimitDegrees=5}
 {bone1Name=dyn_hair_top_bck_01_l, bone2Name=head, swing1LimitDegrees=15, swing2LimitDegrees=15, twistLimitDegrees=5}
 {bone1Name=dyn_hair_top_bck_02_l, bone2Name=dyn_hair_top_bck_01_l, swing1LimitDegrees=15, swing2LimitDegrees=15, twistLimitDegrees=5}
 {bone1Name=dyn_hair_top_bck_03_l, bone2Name=dyn_hair_top_bck_02_l, swing1LimitDegrees=15, swing2LimitDegrees=15, twistLimitDegrees=5}
 {bone1Name=dyn_hair_top_bck_01, bone2Name=head, swing1LimitDegrees=15, swing2LimitDegrees=15, twistLimitDegrees=5}
 {bone1Name=dyn_hair_top_bck_02, bone2Name=dyn_hair_top_bck_01, swing1LimitDegrees=15, swing2LimitDegrees=15, twistLimitDegrees=5}
 {bone1Name=dyn_hair_top_bck_03, bone2Name=dyn_hair_top_bck_02, swing1LimitDegrees=15, swing2LimitDegrees=15, twistLimitDegrees=5}
 {bone1Name=dyn_hair_top_frt_01_r, bone2Name=head, swing1LimitDegrees=25, swing2LimitDegrees=25, twistLimitDegrees=5}
 {bone1Name=dyn_hair_top_frt_02_r, bone2Name=dyn_hair_top_frt_01_r, swing1LimitDegrees=35, swing2LimitDegrees=35, twistLimitDegrees=5}
 {bone1Name=dyn_hair_top_side_01_r, bone2Name=head, swing1LimitDegrees=25, swing2LimitDegrees=25, twistLimitDegrees=5}
 {bone1Name=dyn_hair_top_side_02_r, bone2Name=dyn_hair_top_side_01_r, swing1LimitDegrees=30, swing2LimitDegrees=30, twistLimitDegrees=5}
 {bone1Name=dyn_hair_top_bck_01_r, bone2Name=head, swing1LimitDegrees=15, swing2LimitDegrees=15, twistLimitDegrees=5}
 {bone1Name=dyn_hair_top_bck_02_r, bone2Name=dyn_hair_top_bck_01_r, swing1LimitDegrees=15, swing2LimitDegrees=15, twistLimitDegrees=5}
 {bone1Name=dyn_hair_top_bck_03_r, bone2Name=dyn_hair_top_bck_02_r, swing1LimitDegrees=15, swing2LimitDegrees=15, twistLimitDegrees=5}
 {
  bone1Name=dyn_hair_bck_01_r
  bone2Name=spine_05
  swing1LimitDegrees=45
  swing1Motion=Free
  swing2LimitDegrees=45
  swing2Motion=Free
  twistLimitDegrees=45
  twistMotion=Free
 }
 {
  bone1Name=dyn_hair_bck_01
  bone2Name=spine_05
  swing1LimitDegrees=45
  swing1Motion=Free
  swing2LimitDegrees=45
  swing2Motion=Free
  twistLimitDegrees=45
  twistMotion=Free
 }
 {
  bone1Name=dyn_hair_bck_01_l
  bone2Name=spine_05
  swing1LimitDegrees=45
  swing1Motion=Free
  swing2LimitDegrees=45
  swing2Motion=Free
  twistLimitDegrees=45
  twistMotion=Free
 }
 {
  bone1Name=dyn_hair_side_01_l
  bone2Name=spine_05
  swing1LimitDegrees=45
  swing1Motion=Free
  swing2LimitDegrees=45
  swing2Motion=Free
  twistLimitDegrees=45
  twistMotion=Free
 }
 {
  bone1Name=dyn_hair_side_01_r
  bone2Name=spine_05
  swing1LimitDegrees=45
  swing1Motion=Free
  swing2LimitDegrees=45
  swing2Motion=Free
  twistLimitDegrees=45
  twistMotion=Free
 }
]
